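# Physics asset for the cloak on the female medium Wayfare player body, defining bodies and joints for its dynamic bone chain.
# Five capsules along spine_01 to spine_05 form the torso volume, linked by free-swing, free-twist constraints.
# Three offset, progressively smaller capsules on dyn_cloak bones form a chain hanging from spine_04.
# Cloak joint limits are tightest at the root, with no twist allowance, and widen down the chain.
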PhysicsAsset /BRCosmetics/Characters/Player/Female/Medium/Bodies/F_MED_Wayfare/Meshes/F_MED_Wayfare_Cloak_Physics
bodies {
 dyn_cloak_01 [{shapeType=Capsule, length=10.1833, radius=5.7025, shapeName=dyn_cloak_01_capsule, center=<3.387 -5 0>}]
 dyn_cloak_02 [{shapeType=Capsule, length=6.3135, radius=4.2669, shapeName=dyn_cloak_02_capsule, center=<3.387 -3.6 0>}]
 dyn_cloak_03 [{shapeType=Capsule, length=3.7783, radius=4.0187, shapeName=dyn_cloak_03_capsule, center=<3.387 -3.6 0>}]
 spine_01 [{shapeType=Capsule, length=4.3128, radius=10.241, shapeName=spine_01_capsule}]
 spine_02 [{shapeType=Capsule, length=4.3979, radius=8.8011, shapeName=spine_02_capsule}]
 spine_03 [{shapeType=Capsule, length=5.5487, radius=9.6628, shapeName=spine_03_capsule}]
 spine_04 [{shapeType=Capsule, length=6.9716, radius=10.5183, shapeName=spine_04_capsule}]
 spine_05 [{shapeType=Capsule, length=9.9349, radius=10.895, shapeName=spine_05_capsule}]
}
constraints [
 {
  bone1Name=spine_02
  bone2Name=spine_01
  swing1LimitDegrees=45
  swing1Motion=Free
  swing2LimitDegrees=45
  swing2Motion=Free
  twistLimitDegrees=45
  twistMotion=Free
 }
 {
  bone1Name=spine_03
  bone2Name=spine_02
  swing1LimitDegrees=45
  swing1Motion=Free
  swing2LimitDegrees=45
  swing2Motion=Free
  twistLimitDegrees=45
  twistMotion=Free
 }
 {
  bone1Name=spine_04
  bone2Name=spine_03
  swing1LimitDegrees=45
  swing1Motion=Free
  swing2LimitDegrees=45
  swing2Motion=Free
  twistLimitDegrees=45
  twistMotion=Free
 }
 {
  bone1Name=spine_05
  bone2Name=spine_04
  swing1LimitDegrees=45
  swing1Motion=Free
  swing2LimitDegrees=45
  swing2Motion=Free
  twistLimitDegrees=45
  twistMotion=Free
 }
 {bone1Name=dyn_cloak_01, bone2Name=spine_04, swing1LimitDegrees=25, swing2LimitDegrees=10, twistLimitDegrees=0}
 {bone1Name=dyn_cloak_02, bone2Name=dyn_cloak_01, swing1LimitDegrees=30, swing2LimitDegrees=20, twistLimitDegrees=10}
 {bone1Name=dyn_cloak_03, bone2Name=dyn_cloak_02, swing1LimitDegrees=35, swing2LimitDegrees=20, twistLimitDegrees=10}
]
