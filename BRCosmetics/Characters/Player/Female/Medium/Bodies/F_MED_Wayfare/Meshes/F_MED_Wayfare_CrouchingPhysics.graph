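# Crouch-specific physics asset for the female medium Wayfare body, built to simulate its skirt and dangling accessories.
# Capsules cover the pelvis, lower spine and legs, while the root adds two large volumes and a huge floor-like sphere below it.
# Six skirt chains hang from free-constrained master bones on the thighs, and three accessory chains from the right thigh.
# Chain links use tight swing limits and zero-degree twist limits.
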PhysicsAsset /BRCosmetics/Characters/Player/Female/Medium/Bodies/F_MED_Wayfare/Meshes/F_MED_Wayfare_CrouchingPhysics
bodies {
 calf_l [{shapeType=Capsule, length=40.8697, radius=12.2204, shapeName=calf_l_capsule, center=<-24.6418 0 0>, rotation=<rot 90 171.9277 177.6106>}]
 calf_r [{shapeType=Capsule, length=39.55, radius=13.54, shapeName=calf_l_capsule, center=<24.6418 0 0>, rotation=<rot 90 -7.6507 178.0322>}]
 dyn_accessory_A_01 [{shapeType=Capsule, length=9.166, radius=0.7756, shapeName=dyn_accessory_A_01_capsule, center=<5.5172 0 0>, rotation=<rot 90 0 0>}]
 dyn_accessory_A_02 [{
   shapeType=Capsule
   length=1.3865
   radius=3.1185
   shapeName=dyn_accessory_A_02_capsule
   center=<1.6468 -0.0624 0.0214>
   rotation=<rot 70.9866 89.9607 -168.5144>
  }]
 dyn_accessory_A_MSTR [{shapeType=Capsule, length=0.1, radius=0.1, shapeName=dyn_accessory_A_MSTR_capsule}]
 dyn_accessory_B_01 [{shapeType=Capsule, length=2.6061, radius=0.7091, shapeName=dyn_accessory_B_01_capsule, center=<1.8649 0.0335 0.315>, rotation=<rot 90 0 0>}]
 dyn_accessory_B_02 [{
   shapeType=Capsule
   length=2.0888
   radius=2.4519
   shapeName=dyn_accessory_B_02_capsule
   center=<1.5757 0 0.7>
   rotation=<rot 0 -0.6348 -89.0533>
  }]
 dyn_accessory_B_MSTR [{shapeType=Capsule, length=0.1, radius=0.1, shapeName=dyn_accessory_B_MSTR_capsule}]
 dyn_accessory_C_01 [{
   shapeType=Capsule
   length=14.3172
   radius=1.6858
   shapeName=dyn_accessory_C_01_capsule
   center=<2.5818 -0.5382 -1.4582>
   rotation=<rot 0 87.5028 -89.1065>
  }]
 dyn_accessory_C_MSTR [{shapeType=Capsule, length=0.1, radius=0.1, shapeName=dyn_accessory_C_MSTR_capsule}]
 dyn_skirt_bk_l_01 [{shapeType=Capsule, length=6.7981, radius=7.64, shapeName=dyn_skirt_bk_l_01_capsule, center=<3 4 8>, rotation=<rot 66.4023 89.9998 -0.0002>}]
 dyn_skirt_bk_l_02 [{shapeType=Capsule, length=8.6633, radius=8.9872, shapeName=dyn_skirt_bk_l_02_capsule, center=<8 6 8>, rotation=<rot 66.4023 89.9998 0>}]
 dyn_skirt_bk_l_03 [{
   shapeType=Capsule
   length=10.4888
   radius=9.8759
   shapeName=dyn_skirt_bk_l_03_capsule
   center=<8 4 8>
   rotation=<rot 66.4023 89.9999 -0.0002>
  }]
 dyn_skirt_bk_l_04 [{shapeType=Capsule, length=29.7682, radius=9.3173, shapeName=dyn_skirt_bk_l_04_capsule, center=<5 4 3>, rotation=<rot 69 90 -0.0004>}]
 dyn_skirt_bk_l_05 [{shapeType=Capsule, length=34.945, radius=9.6315, shapeName=dyn_skirt_bk_l_05_capsule, center=<5 4 3>, rotation=<rot 69 90 0>}]
 dyn_skirt_bk_l_06 [{shapeType=Capsule, length=34.945, radius=9.6315, shapeName=dyn_skirt_bk_l_06_capsule, center=<5 4 3>, rotation=<rot 69 90 0>}]
 dyn_skirt_bk_l_MSTR [{shapeType=Capsule, length=1, radius=1, shapeName=dyn_skirt_bk_l_MSTR_capsule, rotation=<rot 90 0 0>}]
 dyn_skirt_bk_r_01 [{
   shapeType=Capsule
   length=6.7288
   radius=7.7092
   shapeName=dyn_skirt_bk_r_01_capsule
   center=<-3 -4 -8>
   rotation=<rot 66.4023 89.9998 -0.0002>
  }]
 dyn_skirt_bk_r_02 [{
   shapeType=Capsule
   length=8.6633
   radius=8.9872
   shapeName=dyn_skirt_bk_r_02_capsule
   center=<-8 -6 -8>
   rotation=<rot 66.4023 89.9998 -0.0002>
  }]
 dyn_skirt_bk_r_03 [{
   shapeType=Capsule
   length=17.1529
   radius=8.7435
   shapeName=dyn_skirt_bk_r_03_capsule
   center=<-8 -4 -8>
   rotation=<rot 66.4023 89.9999 -0.0002>
  }]
 dyn_skirt_bk_r_04 [{
   shapeType=Capsule
   length=29.4887
   radius=9.6727
   shapeName=dyn_skirt_bk_r_04_capsule
   center=<-5 0 -8>
   rotation=<rot 71.0469 89.9996 2.9131>
  }]
 dyn_skirt_bk_r_05 [{shapeType=Capsule, length=38.9212, radius=9.6555, shapeName=dyn_skirt_bk_r_05_capsule, center=<-5 0 -8>, rotation=<rot -0.0002 0 111.6332>}]
 dyn_skirt_bk_r_06 [{shapeType=Capsule, length=38.9212, radius=9.6555, shapeName=dyn_skirt_bk_r_06_capsule, center=<-5 0 -8>, rotation=<rot -0.0002 0 111.6332>}]
 dyn_skirt_bk_r_MSTR [{shapeType=Capsule, length=1, radius=1, shapeName=dyn_skirt_bk_r_MSTR_capsule, rotation=<rot 90 0 0>}]
 dyn_skirt_front_l_01 [{shapeType=Capsule, length=12.4958, radius=3.9658, shapeName=dyn_skirt_front_l_01_capsule, center=<11.9282 0 8>, rotation=<rot 90 0 0>}]
 dyn_skirt_front_l_02 [{shapeType=Capsule, length=8.6375, radius=6.0638, shapeName=dyn_skirt_front_l_02_capsule, center=<3.5 0 5.6735>, rotation=<rot 90 0 0>}]
 dyn_skirt_front_l_03 [{shapeType=Capsule, length=11.2725, radius=9.6059, shapeName=dyn_skirt_front_l_03_capsule, center=<7.0013 0 8>, rotation=<rot 90 0 0>}]
 dyn_skirt_front_l_04 [{shapeType=Capsule, length=12.2641, radius=9.5439, shapeName=dyn_skirt_front_l_04_capsule, center=<7.0013 0 8>, rotation=<rot 90 0 0>}]
 dyn_skirt_front_l_05 [{shapeType=Capsule, length=22.9898, radius=9, shapeName=dyn_skirt_front_l_05_capsule, center=<7.0013 0 8>, rotation=<rot 90 90 0>}]
 dyn_skirt_front_l_06 [{shapeType=Capsule, length=22.9898, radius=9, shapeName=dyn_skirt_front_l_06_capsule, center=<7.0013 0 8>, rotation=<rot 90 90 0>}]
 dyn_skirt_front_l_MSTR [{shapeType=Capsule, length=1, radius=1, shapeName=dyn_skirt_front_l_MSTR_capsule, rotation=<rot 90 0 0>}]
 dyn_skirt_front_r_01 [{shapeType=Capsule, length=9.6011, radius=9.3989, shapeName=dyn_skirt_front_r_01_capsule, center=<-5.948 0 -8>, rotation=<rot 90 0 0>}]
 dyn_skirt_front_r_02 [{shapeType=Capsule, length=9.6011, radius=9.3989, shapeName=dyn_skirt_front_r_02_capsule, center=<-5.948 0 -8>, rotation=<rot 90 0 0>}]
 dyn_skirt_front_r_03 [{shapeType=Capsule, length=9.6011, radius=9.3989, shapeName=dyn_skirt_front_r_03_capsule, center=<-5.948 0 -8>, rotation=<rot 90 0 0>}]
 dyn_skirt_front_r_04 [{shapeType=Capsule, length=9.6011, radius=9.3989, shapeName=dyn_skirt_front_r_04_capsule, center=<-5.948 0 -8>, rotation=<rot 90 0 0>}]
 dyn_skirt_front_r_05 [{shapeType=Capsule, length=22.9898, radius=9, shapeName=dyn_skirt_front_r_05_capsule, center=<-5.948 0 -8>, rotation=<rot 90 90 0>}]
 dyn_skirt_front_r_06 [{shapeType=Capsule, length=22.9898, radius=9, shapeName=dyn_skirt_front_r_06_capsule, center=<-5.948 0 -8>, rotation=<rot 90 90 0>}]
 dyn_skirt_front_r_MSTR [{shapeType=Capsule, length=1, radius=1, shapeName=dyn_skirt_front_r_MSTR_capsule, rotation=<rot 90 0 0>}]
 dyn_skirt_l_01 [{shapeType=Capsule, length=15.3397, radius=3.6603, shapeName=dyn_skirt_l_01_capsule, center=<11.9282 0 8>, rotation=<rot 90 0 0>}]
 dyn_skirt_l_02 [{shapeType=Capsule, length=8.5626, radius=6.1387, shapeName=dyn_skirt_l_02_capsule, center=<3.5 0 5.6735>, rotation=<rot 90 0 0>}]
 dyn_skirt_l_03 [{shapeType=Capsule, length=9.4398, radius=9.5602, shapeName=dyn_skirt_l_03_capsule, center=<7.0013 0 8>, rotation=<rot 90 0 0>}]
 dyn_skirt_l_04 [{shapeType=Capsule, length=35.3737, radius=9, shapeName=dyn_skirt_l_04_capsule, center=<7.0013 0 8>, rotation=<rot 68.9719 -90 180>}]
 dyn_skirt_l_05 [{shapeType=Capsule, length=32.6803, radius=9, shapeName=dyn_skirt_l_05_capsule, center=<7.0013 0 8>, rotation=<rot 68.9719 -90 180>}]
 dyn_skirt_l_06 [{shapeType=Capsule, length=32.6803, radius=9, shapeName=dyn_skirt_l_06_capsule, center=<7.0013 0 8>, rotation=<rot 68.9719 -90 180>}]
 dyn_skirt_l_MSTR [{shapeType=Capsule, length=1, radius=1, shapeName=dyn_skirt_l_MSTR_capsule, rotation=<rot 90 0 0>}]
 dyn_skirt_r_01 [{shapeType=Capsule, length=10, radius=9, shapeName=dyn_skirt_r_01_capsule, center=<-5.948 0 -8>, rotation=<rot 90 0 0>}]
 dyn_skirt_r_02 [{shapeType=Capsule, length=9.4398, radius=9.5602, shapeName=dyn_skirt_r_02_capsule, center=<-5.948 0 -8>, rotation=<rot 90 0 0>}]
 dyn_skirt_r_03 [{shapeType=Capsule, length=9.4398, radius=9.5602, shapeName=dyn_skirt_r_03_capsule, center=<-5.948 0 -8>, rotation=<rot 90 0 0>}]
 dyn_skirt_r_04 [{shapeType=Capsule, length=34.1832, radius=9, shapeName=dyn_skirt_r_04_capsule, center=<-5.948 0 -8>, rotation=<rot 69.1316 -89.9999 180>}]
 dyn_skirt_r_05 [{shapeType=Capsule, length=31.4898, radius=9, shapeName=dyn_skirt_r_05_capsule, center=<-5.948 0 -8>, rotation=<rot 69.1316 -89.9999 180>}]
 dyn_skirt_r_06 [{shapeType=Capsule, length=31.4898, radius=9, shapeName=dyn_skirt_r_06_capsule, center=<-5.948 0 -8>, rotation=<rot 69.1316 -89.9999 180>}]
 dyn_skirt_r_MSTR [{shapeType=Capsule, length=1, radius=1, shapeName=dyn_skirt_r_MSTR_capsule, rotation=<rot 90 0 0>}]
 foot_l [{shapeType=Capsule, length=16.5331, radius=9.9698, shapeName=foot_l_capsule, center=<-3.2449 6.0456 0>, rotation=<rot 90 180 90>}]
 foot_r [{shapeType=Capsule, length=20.6248, radius=9.7591, shapeName=foot_l_capsule, center=<3.2449 -6.0456 0>, rotation=<rot 90 -165.9638 -75.9638>}]
 pelvis [{shapeType=Capsule, length=6.7014, radius=10.7296, shapeName=pelvis_capsule, center=<0 -1.391 0>}]
 root [
  {shapeType=Capsule, length=0, radius=300, shapeName=root_capsule, center=<0 -9.6788 -287.2379>}
  {
   shapeType=Capsule
   length=29.8921
   radius=23.7573
   shapeName=root_capsule
   center=<-12.1516 -43.8975 -3.0879>
   rotation=<rot 80.3873 175.3621 -175.5078>
  }
  {
   shapeType=Capsule
   length=37.5342
   radius=23.4523
   shapeName=root_capsule
   center=<-0.8597 -22.8786 1.2902>
   rotation=<rot -52.4403 94.724 -9.0379>
  }
 ]
 spine_01 [{shapeType=Capsule, length=5.9326, radius=8.4081, shapeName=spine_01_capsule}]
 thigh_l [{
   shapeType=Capsule
   length=32.8974
   radius=11.586
   shapeName=thigh_l_capsule
   center=<-24.8541 -2.4007 2.6623>
   rotation=<rot 85.5057 179.9991 178.8682>
  }]
 thigh_r [{
   shapeType=Capsule
   length=32.8974
   radius=11.586
   shapeName=thigh_l_capsule
   center=<24.8541 2.4007 -2.6623>
   rotation=<rot 85.5057 179.9991 -1.1318>
  }]
}
constraints [
 {
  bone1Name=thigh_l
  bone2Name=pelvis
  swing1LimitDegrees=45
  swing1Motion=Free
  swing2LimitDegrees=45
  swing2Motion=Free
  twistLimitDegrees=45
  twistMotion=Free
 }
 {
  bone1Name=pelvis
  bone2Name=root
  swing1LimitDegrees=45
  swing1Motion=Free
  swing2LimitDegrees=45
  swing2Motion=Free
  twistLimitDegrees=45
  twistMotion=Free
 }
 {
  bone1Name=spine_01
  bone2Name=pelvis
  swing1LimitDegrees=45
  swing1Motion=Free
  swing2LimitDegrees=45
  swing2Motion=Free
  twistLimitDegrees=45
  twistMotion=Free
 }
 {
  bone1Name=calf_l
  bone2Name=thigh_l
  swing1LimitDegrees=45
  swing1Motion=Free
  swing2LimitDegrees=45
  swing2Motion=Free
  twistLimitDegrees=45
  twistMotion=Free
 }
 {
  bone1Name=foot_l
  bone2Name=calf_l
  swing1LimitDegrees=45
  swing1Motion=Free
  swing2LimitDegrees=45
  swing2Motion=Free
  twistLimitDegrees=45
  twistMotion=Free
 }
 {
  bone1Name=calf_r
  bone2Name=thigh_r
  swing1LimitDegrees=45
  swing1Motion=Free
  swing2LimitDegrees=45
  swing2Motion=Free
  twistLimitDegrees=45
  twistMotion=Free
 }
 {
  bone1Name=foot_r
  bone2Name=calf_r
  swing1LimitDegrees=45
  swing1Motion=Free
  swing2LimitDegrees=45
  swing2Motion=Free
  twistLimitDegrees=45
  twistMotion=Free
 }
 {
  bone1Name=dyn_skirt_front_l_MSTR
  bone2Name=thigh_l
  swing1LimitDegrees=45
  swing1Motion=Free
  swing2LimitDegrees=45
  swing2Motion=Free
  twistLimitDegrees=45
  twistMotion=Free
 }
 {bone1Name=dyn_skirt_front_l_01, bone2Name=dyn_skirt_front_l_MSTR, swing1LimitDegrees=20, swing2LimitDegrees=30, twistLimitDegrees=0}
 {bone1Name=dyn_skirt_front_l_02, bone2Name=dyn_skirt_front_l_01, swing1LimitDegrees=20, swing2LimitDegrees=30, twistLimitDegrees=0}
 {bone1Name=dyn_skirt_front_l_03, bone2Name=dyn_skirt_front_l_02, swing1LimitDegrees=20, swing2LimitDegrees=35, twistLimitDegrees=0}
 {bone1Name=dyn_skirt_front_l_04, bone2Name=dyn_skirt_front_l_03, swing1LimitDegrees=20, swing2LimitDegrees=30, twistLimitDegrees=0}
 {bone1Name=dyn_skirt_front_l_05, bone2Name=dyn_skirt_front_l_04, swing1LimitDegrees=20, swing2LimitDegrees=45, twistLimitDegrees=0}
 {bone1Name=dyn_skirt_front_l_06, bone2Name=dyn_skirt_front_l_05, swing1LimitDegrees=20, swing2LimitDegrees=45, twistLimitDegrees=0}
 {
  bone1Name=dyn_skirt_bk_l_MSTR
  bone2Name=thigh_l
  swing1LimitDegrees=45
  swing1Motion=Free
  swing2LimitDegrees=45
  swing2Motion=Free
  twistLimitDegrees=45
  twistMotion=Free
 }
 {bone1Name=dyn_skirt_bk_l_01, bone2Name=dyn_skirt_bk_l_MSTR, swing1LimitDegrees=7, swing2LimitDegrees=20, twistLimitDegrees=0}
 {bone1Name=dyn_skirt_bk_l_02, bone2Name=dyn_skirt_bk_l_01, swing1LimitDegrees=10, swing2LimitDegrees=20, twistLimitDegrees=0}
 {bone1Name=dyn_skirt_bk_l_03, bone2Name=dyn_skirt_bk_l_02, swing1LimitDegrees=10, swing2LimitDegrees=45, twistLimitDegrees=0}
 {bone1Name=dyn_skirt_bk_l_04, bone2Name=dyn_skirt_bk_l_03, swing1LimitDegrees=20, swing2LimitDegrees=45, twistLimitDegrees=0}
 {bone1Name=dyn_skirt_bk_l_05, bone2Name=dyn_skirt_bk_l_04, swing1LimitDegrees=20, swing2LimitDegrees=45, twistLimitDegrees=0}
 {bone1Name=dyn_skirt_bk_l_06, bone2Name=dyn_skirt_bk_l_05, swing1LimitDegrees=20, swing2LimitDegrees=45, twistLimitDegrees=0}
 {
  bone1Name=dyn_skirt_l_MSTR
  bone2Name=thigh_l
  swing1LimitDegrees=45
  swing1Motion=Free
  swing2LimitDegrees=45
  swing2Motion=Free
  twistLimitDegrees=45
  twistMotion=Free
 }
 {bone1Name=dyn_skirt_l_01, bone2Name=dyn_skirt_l_MSTR, swing1LimitDegrees=20, swing2LimitDegrees=20, twistLimitDegrees=0}
 {bone1Name=dyn_skirt_l_02, bone2Name=dyn_skirt_l_01, swing1LimitDegrees=20, swing2LimitDegrees=20, twistLimitDegrees=0}
 {bone1Name=dyn_skirt_l_03, bone2Name=dyn_skirt_l_02, swing1LimitDegrees=20, swing2LimitDegrees=20, twistLimitDegrees=0}
 {bone1Name=dyn_skirt_l_04, bone2Name=dyn_skirt_l_03, swing1LimitDegrees=20, swing2LimitDegrees=20, twistLimitDegrees=0}
 {bone1Name=dyn_skirt_l_05, bone2Name=dyn_skirt_l_04, swing1LimitDegrees=20, swing2LimitDegrees=20, twistLimitDegrees=0}
 {bone1Name=dyn_skirt_l_06, bone2Name=dyn_skirt_l_05, swing1LimitDegrees=20, swing2LimitDegrees=20, twistLimitDegrees=0}
 {
  bone1Name=dyn_skirt_front_r_MSTR
  bone2Name=thigh_r
  swing1LimitDegrees=45
  swing1Motion=Free
  swing2LimitDegrees=45
  swing2Motion=Free
  twistLimitDegrees=45
  twistMotion=Free
 }
 {bone1Name=dyn_skirt_front_r_01, bone2Name=dyn_skirt_front_r_MSTR, swing1LimitDegrees=20, swing2LimitDegrees=20, twistLimitDegrees=0}
 {bone1Name=dyn_skirt_front_r_02, bone2Name=dyn_skirt_front_r_01, swing1LimitDegrees=20, swing2LimitDegrees=20, twistLimitDegrees=0}
 {bone1Name=dyn_skirt_front_r_03, bone2Name=dyn_skirt_front_r_02, swing1LimitDegrees=20, swing2LimitDegrees=20, twistLimitDegrees=0}
 {bone1Name=dyn_skirt_front_r_04, bone2Name=dyn_skirt_front_r_03, swing1LimitDegrees=20, swing2LimitDegrees=20, twistLimitDegrees=0}
 {bone1Name=dyn_skirt_front_r_05, bone2Name=dyn_skirt_front_r_04, swing1LimitDegrees=20, swing2LimitDegrees=45, twistLimitDegrees=0}
 {bone1Name=dyn_skirt_front_r_06, bone2Name=dyn_skirt_front_r_05, swing1LimitDegrees=20, swing2LimitDegrees=45, twistLimitDegrees=0}
 {
  bone1Name=dyn_skirt_bk_r_MSTR
  bone2Name=thigh_r
  swing1LimitDegrees=45
  swing1Motion=Free
  swing2LimitDegrees=45
  swing2Motion=Free
  twistLimitDegrees=45
  twistMotion=Free
 }
 {bone1Name=dyn_skirt_bk_r_01, bone2Name=dyn_skirt_bk_r_MSTR, swing1LimitDegrees=7, swing2LimitDegrees=20, twistLimitDegrees=0}
 {bone1Name=dyn_skirt_bk_r_02, bone2Name=dyn_skirt_bk_r_01, swing1LimitDegrees=10, swing2LimitDegrees=20, twistLimitDegrees=0}
 {bone1Name=dyn_skirt_bk_r_03, bone2Name=dyn_skirt_bk_r_02, swing1LimitDegrees=10, swing2LimitDegrees=45, twistLimitDegrees=0}
 {bone1Name=dyn_skirt_bk_r_04, bone2Name=dyn_skirt_bk_r_03, swing1LimitDegrees=20, swing2LimitDegrees=45, twistLimitDegrees=0}
 {bone1Name=dyn_skirt_bk_r_05, bone2Name=dyn_skirt_bk_r_04, swing1LimitDegrees=20, swing2LimitDegrees=45, twistLimitDegrees=0}
 {bone1Name=dyn_skirt_bk_r_06, bone2Name=dyn_skirt_bk_r_05, swing1LimitDegrees=20, swing2LimitDegrees=45, twistLimitDegrees=0}
 {
  bone1Name=dyn_skirt_r_MSTR
  bone2Name=thigh_r
  swing1LimitDegrees=45
  swing1Motion=Free
  swing2LimitDegrees=45
  swing2Motion=Free
  twistLimitDegrees=45
  twistMotion=Free
 }
 {bone1Name=dyn_skirt_r_01, bone2Name=dyn_skirt_r_MSTR, swing1LimitDegrees=20, swing2LimitDegrees=20, twistLimitDegrees=0}
 {bone1Name=dyn_skirt_r_02, bone2Name=dyn_skirt_r_01, swing1LimitDegrees=20, swing2LimitDegrees=20, twistLimitDegrees=0}
 {bone1Name=dyn_skirt_r_03, bone2Name=dyn_skirt_r_02, swing1LimitDegrees=20, swing2LimitDegrees=20, twistLimitDegrees=0}
 {bone1Name=dyn_skirt_r_04, bone2Name=dyn_skirt_r_03, swing1LimitDegrees=20, swing2LimitDegrees=20, twistLimitDegrees=0}
 {bone1Name=dyn_skirt_r_05, bone2Name=dyn_skirt_r_04, swing1LimitDegrees=20, swing2LimitDegrees=20, twistLimitDegrees=0}
 {bone1Name=dyn_skirt_r_06, bone2Name=dyn_skirt_r_05, swing1LimitDegrees=20, swing2LimitDegrees=20, twistLimitDegrees=0}
 {
  bone1Name=thigh_r
  bone2Name=pelvis
  swing1LimitDegrees=45
  swing1Motion=Free
  swing2LimitDegrees=45
  swing2Motion=Free
  twistLimitDegrees=45
  twistMotion=Free
 }
 {bone1Name=dyn_accessory_C_01, bone2Name=dyn_accessory_C_MSTR, swing1LimitDegrees=30, swing2LimitDegrees=45, twistLimitDegrees=0}
 {bone1Name=dyn_accessory_B_01, bone2Name=dyn_accessory_B_MSTR, swing1LimitDegrees=30, swing2LimitDegrees=45, twistLimitDegrees=0}
 {bone1Name=dyn_accessory_B_02, bone2Name=dyn_accessory_B_01, swing1LimitDegrees=20, swing2LimitDegrees=20, twistLimitDegrees=0}
 {bone1Name=dyn_accessory_A_01, bone2Name=dyn_accessory_A_MSTR, swing1LimitDegrees=35, swing2LimitDegrees=45, twistLimitDegrees=0}
 {bone1Name=dyn_accessory_A_02, bone2Name=dyn_accessory_A_01, swing1LimitDegrees=20, swing2LimitDegrees=20, twistLimitDegrees=0}
 {
  bone1Name=dyn_accessory_C_MSTR
  bone2Name=thigh_r
  swing1LimitDegrees=45
  swing1Motion=Free
  swing2LimitDegrees=45
  swing2Motion=Free
  twistLimitDegrees=45
  twistMotion=Free
 }
 {
  bone1Name=dyn_accessory_B_MSTR
  bone2Name=thigh_r
  swing1LimitDegrees=45
  swing1Motion=Free
  swing2LimitDegrees=45
  swing2Motion=Free
  twistLimitDegrees=45
  twistMotion=Free
 }
 {
  bone1Name=dyn_accessory_A_MSTR
  bone2Name=thigh_r
  swing1LimitDegrees=45
  swing1Motion=Free
  swing2LimitDegrees=45
  swing2Motion=Free
  twistLimitDegrees=45
  twistMotion=Free
 }
]
