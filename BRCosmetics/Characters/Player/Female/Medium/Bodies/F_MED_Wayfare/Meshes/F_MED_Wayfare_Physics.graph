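# Physics asset for the female medium Wayfare cosmetic body, defining secondary motion for its skirt and three dangling accessories.
# Capsules on the legs, pelvis and lower spine, a thin left-thigh box and a huge capsule below the root form the collision body.
# Six skirt chains and three right-thigh accessory chains use tight swing limits and zero twist.
# Their master bones join the thighs with free constraints, except back-right skirt, which has none.
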PhysicsAsset /BRCosmetics/Characters/Player/Female/Medium/Bodies/F_MED_Wayfare/Meshes/F_MED_Wayfare_Physics
bodies {
 calf_l [
  {shapeType=Capsule, length=42.9166, radius=12.1734, shapeName=calf_l_capsule, center=<-24.8184 -1.7754 0>, rotation=<rot 90 180 -174.3172>}
  {shapeType=Capsule, length=16.25, radius=8.75, shapeName=calf_l_capsule, center=<-10.1679 -2.3046 0>, rotation=<rot 90 180 175.0654>}
 ]
 calf_r [
  {shapeType=Capsule, length=43.9166, radius=11.1734, shapeName=calf_l_capsule, center=<24.8461 2.0533 0>, rotation=<rot 90 180 5.6828>}
  {shapeType=Capsule, length=16.25, radius=7.75, shapeName=calf_r_capsule, center=<10.5071 2.5016 0>, rotation=<rot 90 180 171.6286>}
 ]
 dyn_accessory_A_01 [{shapeType=Capsule, length=8.32, radius=1.6216, shapeName=dyn_accessory_A_01_capsule, center=<5.5172 0 0>, rotation=<rot 90 0 0>}]
 dyn_accessory_A_02 [{
   shapeType=Capsule
   length=1.3865
   radius=3.1185
   shapeName=dyn_accessory_A_02_capsule
   center=<1.6468 -0.0624 0.0214>
   rotation=<rot 70.9866 89.9607 -168.5144>
  }]
 dyn_accessory_A_MSTR [{shapeType=Capsule, length=0.1, radius=0.1, shapeName=dyn_accessory_A_MSTR_capsule}]
 dyn_accessory_B_01 [{shapeType=Capsule, length=2.6061, radius=0.7091, shapeName=dyn_accessory_B_01_capsule, center=<1.8649 0.0335 0.315>, rotation=<rot 90 0 0>}]
 dyn_accessory_B_02 [{
   shapeType=Capsule
   length=2.0888
   radius=2.4519
   shapeName=dyn_accessory_B_02_capsule
   center=<1.5757 0 0.7>
   rotation=<rot 0 -0.6348 -89.0533>
  }]
 dyn_accessory_B_MSTR [{shapeType=Capsule, length=0.1, radius=0.1, shapeName=dyn_accessory_B_MSTR_capsule}]
 dyn_accessory_C_01 [{
   shapeType=Capsule
   length=14.3172
   radius=1.6858
   shapeName=dyn_accessory_C_01_capsule
   center=<2.5818 -0.5382 -1.4582>
   rotation=<rot 0 87.5028 -89.1065>
  }]
 dyn_accessory_C_MSTR [{shapeType=Capsule, length=0.1, radius=0.1, shapeName=dyn_accessory_C_MSTR_capsule}]
 dyn_skirt_bk_l_01 [{shapeType=Capsule, length=5.5481, radius=9.64, shapeName=dyn_skirt_bk_l_01_capsule, center=<3 4 8>, rotation=<rot 66.4023 89.9998 -0.0002>}]
 dyn_skirt_bk_l_02 [{shapeType=Capsule, length=6.9133, radius=11.4872, shapeName=dyn_skirt_bk_l_02_capsule, center=<8 6 8>, rotation=<rot 66.4023 89.9998 0>}]
 dyn_skirt_bk_l_03 [{
   shapeType=Capsule
   length=8.9888
   radius=11.3759
   shapeName=dyn_skirt_bk_l_03_capsule
   center=<8 4 8>
   rotation=<rot 66.4023 89.9999 -0.0002>
  }]
 dyn_skirt_bk_l_04 [{shapeType=Capsule, length=14.5183, radius=10.3173, shapeName=dyn_skirt_bk_l_04_capsule, center=<5 0 8>, rotation=<rot 90 90 0>}]
 dyn_skirt_bk_l_05 [{shapeType=Capsule, length=25.0164, radius=12, shapeName=dyn_skirt_bk_l_05_capsule, center=<5 0 8>, rotation=<rot -0.0001 0 107.1553>}]
 dyn_skirt_bk_l_06 [{shapeType=Capsule, length=25.0164, radius=12, shapeName=dyn_skirt_bk_l_06_capsule, center=<5 0 8>, rotation=<rot -0.0002 0 107.1553>}]
 dyn_skirt_bk_l_MSTR [{shapeType=Capsule, length=1, radius=1, shapeName=dyn_skirt_bk_l_MSTR_capsule, rotation=<rot 90 0 0>}]
 dyn_skirt_bk_r_01 [{
   shapeType=Capsule
   length=5.4788
   radius=8.9592
   shapeName=dyn_skirt_bk_r_01_capsule
   center=<-3 -4 -8>
   rotation=<rot 66.4023 89.9998 -0.0002>
  }]
 dyn_skirt_bk_r_02 [{
   shapeType=Capsule
   length=6.9133
   radius=10.7372
   shapeName=dyn_skirt_bk_r_02_capsule
   center=<-8 -6 -8>
   rotation=<rot 66.4023 89.9998 -0.0002>
  }]
 dyn_skirt_bk_r_03 [{
   shapeType=Capsule
   length=10.1212
   radius=10.2435
   shapeName=dyn_skirt_bk_r_03_capsule
   center=<-8 -4 -8>
   rotation=<rot 66.4023 89.9999 -0.0002>
  }]
 dyn_skirt_bk_r_04 [{
   shapeType=Capsule
   length=25.2121
   radius=10.6727
   shapeName=dyn_skirt_bk_r_04_capsule
   center=<-5 -6 -5.5806>
   rotation=<rot 58.934 89.9998 -0.0002>
  }]
 dyn_skirt_bk_r_05 [{shapeType=Capsule, length=33.9212, radius=12, shapeName=dyn_skirt_bk_r_05_capsule, center=<-5 -5 -6>, rotation=<rot 58.934 89.9998 -0.0002>}]
 dyn_skirt_bk_r_06 [{
   shapeType=Capsule
   length=35.1712
   radius=12
   shapeName=dyn_skirt_bk_r_06_capsule
   center=<-5.0806 -5 -6>
   rotation=<rot 58.934 89.9998 -0.0003>
  }]
 dyn_skirt_bk_r_MSTR [{shapeType=Capsule, length=1, radius=1, shapeName=dyn_skirt_bk_r_MSTR_capsule, rotation=<rot 90 0 0>}]
 dyn_skirt_front_l_01 [{shapeType=Capsule, length=8.9254, radius=9, shapeName=dyn_skirt_front_l_01_capsule, center=<11.9282 0 8>, rotation=<rot 90 0 0>}]
 dyn_skirt_front_l_02 [{shapeType=Capsule, length=8.9665, radius=9, shapeName=dyn_skirt_front_l_02_capsule, center=<7.0013 0 8>, rotation=<rot 90 0 0>}]
 dyn_skirt_front_l_03 [{shapeType=Capsule, length=10.7244, radius=9, shapeName=dyn_skirt_front_l_03_capsule, center=<7.0013 0 8>, rotation=<rot 90 0 0>}]
 dyn_skirt_front_l_04 [{shapeType=Capsule, length=12.2641, radius=9, shapeName=dyn_skirt_front_l_04_capsule, center=<7.0013 0 8>, rotation=<rot 90 0 0>}]
 dyn_skirt_front_l_05 [{shapeType=Capsule, length=22.9898, radius=11, shapeName=dyn_skirt_front_l_05_capsule, center=<7.0013 0 8>, rotation=<rot 90 90 0>}]
 dyn_skirt_front_l_06 [{shapeType=Capsule, length=22.9898, radius=11, shapeName=dyn_skirt_front_l_06_capsule, center=<7.0013 0 8>, rotation=<rot 90 90 0>}]
 dyn_skirt_front_l_MSTR [{shapeType=Capsule, length=1, radius=1, shapeName=dyn_skirt_front_l_MSTR_capsule, rotation=<rot 90 0 0>}]
 dyn_skirt_front_r_01 [{shapeType=Capsule, length=9.6011, radius=9.3989, shapeName=dyn_skirt_front_r_01_capsule, center=<-5.948 0 -8>, rotation=<rot 90 0 0>}]
 dyn_skirt_front_r_02 [{shapeType=Capsule, length=9.6011, radius=9.3989, shapeName=dyn_skirt_front_r_02_capsule, center=<-5.948 0 -8>, rotation=<rot 90 0 0>}]
 dyn_skirt_front_r_03 [{shapeType=Capsule, length=9.6011, radius=9.3989, shapeName=dyn_skirt_front_r_03_capsule, center=<-5.948 0 -8>, rotation=<rot 90 0 0>}]
 dyn_skirt_front_r_04 [{shapeType=Capsule, length=9.6011, radius=9.3989, shapeName=dyn_skirt_front_r_04_capsule, center=<-5.948 0 -8>, rotation=<rot 90 0 0>}]
 dyn_skirt_front_r_05 [{shapeType=Capsule, length=22.9898, radius=11, shapeName=dyn_skirt_front_r_05_capsule, center=<-5.948 0 -8>, rotation=<rot 90 90 0>}]
 dyn_skirt_front_r_06 [{shapeType=Capsule, length=22.9898, radius=11, shapeName=dyn_skirt_front_r_06_capsule, center=<-5.948 0 -8>, rotation=<rot 90 90 0>}]
 dyn_skirt_front_r_MSTR [{shapeType=Capsule, length=1, radius=1, shapeName=dyn_skirt_front_r_MSTR_capsule, rotation=<rot 90 0 0>}]
 dyn_skirt_l_01 [{shapeType=Capsule, length=10.75, radius=9, shapeName=dyn_skirt_l_01_capsule, center=<11.9282 0 8>, rotation=<rot 90 0 0>}]
 dyn_skirt_l_02 [{shapeType=Capsule, length=10.6898, radius=9.0602, shapeName=dyn_skirt_l_02_capsule, center=<7.0013 0 8>, rotation=<rot 90 0 0>}]
 dyn_skirt_l_03 [{shapeType=Capsule, length=9.4398, radius=9.5602, shapeName=dyn_skirt_l_03_capsule, center=<7.0013 0 8>, rotation=<rot 90 0 0>}]
 dyn_skirt_l_04 [{shapeType=Capsule, length=25.6832, radius=9, shapeName=dyn_skirt_l_04_capsule, center=<7.0013 0 8>, rotation=<rot 90 90 0>}]
 dyn_skirt_l_05 [{shapeType=Capsule, length=22.9898, radius=11, shapeName=dyn_skirt_l_05_capsule, center=<7.0013 0 8>, rotation=<rot 90 90 0>}]
 dyn_skirt_l_06 [{shapeType=Capsule, length=22.9898, radius=11, shapeName=dyn_skirt_l_06_capsule, center=<7.0013 0 8>, rotation=<rot 90 90 0>}]
 dyn_skirt_l_MSTR [{shapeType=Capsule, length=1, radius=1, shapeName=dyn_skirt_l_MSTR_capsule, rotation=<rot 90 0 0>}]
 dyn_skirt_r_01 [{shapeType=Capsule, length=10, radius=9, shapeName=dyn_skirt_r_01_capsule, center=<-5.948 0 -8>, rotation=<rot 90 0 0>}]
 dyn_skirt_r_02 [{shapeType=Capsule, length=9.4398, radius=9.5602, shapeName=dyn_skirt_r_02_capsule, center=<-5.948 0 -8>, rotation=<rot 90 0 0>}]
 dyn_skirt_r_03 [{shapeType=Capsule, length=9.4398, radius=9.5602, shapeName=dyn_skirt_r_03_capsule, center=<-5.948 0 -8>, rotation=<rot 90 0 0>}]
 dyn_skirt_r_04 [{shapeType=Capsule, length=26.1832, radius=9, shapeName=dyn_skirt_r_04_capsule, center=<-5.948 0 -8>, rotation=<rot 83.1637 -90 179.9999>}]
 dyn_skirt_r_05 [{shapeType=Capsule, length=28.4898, radius=11, shapeName=dyn_skirt_r_05_capsule, center=<-5.948 0 -8>, rotation=<rot 83.1637 -90 179.9999>}]
 dyn_skirt_r_06 [{shapeType=Capsule, length=28.4898, radius=11, shapeName=dyn_skirt_r_06_capsule, center=<-5.948 0 -8>, rotation=<rot 83.1637 -90 179.9999>}]
 dyn_skirt_r_MSTR [{shapeType=Capsule, length=1, radius=1, shapeName=dyn_skirt_r_MSTR_capsule, rotation=<rot 90 0 0>}]
 foot_l [{shapeType=Capsule, length=24.0331, radius=12.4698, shapeName=foot_l_capsule, center=<-3.2449 6.6917 0>, rotation=<rot 90 180 90>}]
 foot_r [{shapeType=Capsule, length=18.2831, radius=9.9698, shapeName=foot_l_capsule, center=<3.2449 -6.0456 0>, rotation=<rot 90 -165.9638 -75.9638>}]
 pelvis [{shapeType=Capsule, length=6.7014, radius=10.7296, shapeName=pelvis_capsule, center=<0 -1.391 0>}]
 root [{shapeType=Capsule, length=1.8091, radius=292, shapeName=root_capsule, center=<0 -9.6788 -287.2379>}]
 spine_01 [{shapeType=Capsule, length=5.9326, radius=8.4081, shapeName=spine_01_capsule}]
 thigh_l [
  {
   shapeType=Capsule
   length=17.3974
   radius=9.586
   shapeName=thigh_l_capsule
   center=<-12.7514 -2.1633 2.1373>
   rotation=<rot 87.0298 179.4175 178.2881>
  }
  {shapeType=Capsule, length=9.75, radius=7, shapeName=thigh_l_capsule, center=<-32.8894 0 0>, rotation=<rot 90 0 0>}
  {shapeType=Box, extentX=36, extentY=37, extentZ=1, shapeName=thigh_l_box, center=<-35.4102 16.4468 6.6167>}
 ]
 thigh_r [
  {
   shapeType=Capsule
   length=17.8974
   radius=10.586
   shapeName=thigh_l_capsule
   center=<12.9095 1.4404 -1.3839>
   rotation=<rot 87.7285 178.8641 0>
  }
  {shapeType=Capsule, length=22.25, radius=7.5, shapeName=thigh_r_capsule, center=<33.9186 0.7474 0>, rotation=<rot 90 0 0>}
 ]
}
constraints [
 {
  bone1Name=thigh_l
  bone2Name=pelvis
  swing1LimitDegrees=45
  swing1Motion=Free
  swing2LimitDegrees=45
  swing2Motion=Free
  twistLimitDegrees=45
  twistMotion=Free
 }
 {
  bone1Name=pelvis
  bone2Name=root
  swing1LimitDegrees=45
  swing1Motion=Free
  swing2LimitDegrees=45
  swing2Motion=Free
  twistLimitDegrees=45
  twistMotion=Free
 }
 {
  bone1Name=spine_01
  bone2Name=pelvis
  swing1LimitDegrees=45
  swing1Motion=Free
  swing2LimitDegrees=45
  swing2Motion=Free
  twistLimitDegrees=45
  twistMotion=Free
 }
 {
  bone1Name=calf_l
  bone2Name=thigh_l
  swing1LimitDegrees=45
  swing1Motion=Free
  swing2LimitDegrees=45
  swing2Motion=Free
  twistLimitDegrees=45
  twistMotion=Free
 }
 {
  bone1Name=foot_l
  bone2Name=calf_l
  swing1LimitDegrees=45
  swing1Motion=Free
  swing2LimitDegrees=45
  swing2Motion=Free
  twistLimitDegrees=45
  twistMotion=Free
 }
 {
  bone1Name=foot_r
  bone2Name=calf_r
  swing1LimitDegrees=45
  swing1Motion=Free
  swing2LimitDegrees=45
  swing2Motion=Free
  twistLimitDegrees=45
  twistMotion=Free
 }
 {
  bone1Name=dyn_skirt_front_l_MSTR
  bone2Name=thigh_l
  swing1LimitDegrees=45
  swing1Motion=Free
  swing2LimitDegrees=45
  swing2Motion=Free
  twistLimitDegrees=45
  twistMotion=Free
 }
 {bone1Name=dyn_skirt_front_l_01, bone2Name=dyn_skirt_front_l_MSTR, swing1LimitDegrees=20, swing2LimitDegrees=20, twistLimitDegrees=0}
 {bone1Name=dyn_skirt_front_l_02, bone2Name=dyn_skirt_front_l_01, swing1LimitDegrees=20, swing2LimitDegrees=20, twistLimitDegrees=0}
 {bone1Name=dyn_skirt_front_l_03, bone2Name=dyn_skirt_front_l_02, swing1LimitDegrees=20, swing2LimitDegrees=20, twistLimitDegrees=0}
 {bone1Name=dyn_skirt_front_l_04, bone2Name=dyn_skirt_front_l_03, swing1LimitDegrees=20, swing2LimitDegrees=20, twistLimitDegrees=0}
 {bone1Name=dyn_skirt_front_l_05, bone2Name=dyn_skirt_front_l_04, swing1LimitDegrees=20, swing2LimitDegrees=45, twistLimitDegrees=0}
 {bone1Name=dyn_skirt_front_l_06, bone2Name=dyn_skirt_front_l_05, swing1LimitDegrees=20, swing2LimitDegrees=45, twistLimitDegrees=0}
 {
  bone1Name=dyn_skirt_bk_l_MSTR
  bone2Name=thigh_l
  swing1LimitDegrees=45
  swing1Motion=Free
  swing2LimitDegrees=45
  swing2Motion=Free
  twistLimitDegrees=45
  twistMotion=Free
 }
 {bone1Name=dyn_skirt_bk_l_01, bone2Name=dyn_skirt_bk_l_MSTR, swing1LimitDegrees=7, swing2LimitDegrees=20, twistLimitDegrees=0}
 {bone1Name=dyn_skirt_bk_l_02, bone2Name=dyn_skirt_bk_l_01, swing1LimitDegrees=6, swing2LimitDegrees=20, twistLimitDegrees=0}
 {bone1Name=dyn_skirt_bk_l_03, bone2Name=dyn_skirt_bk_l_02, swing1LimitDegrees=6, swing2LimitDegrees=20, twistLimitDegrees=0}
 {bone1Name=dyn_skirt_bk_l_04, bone2Name=dyn_skirt_bk_l_03, swing1LimitDegrees=6, swing2LimitDegrees=20, twistLimitDegrees=0}
 {bone1Name=dyn_skirt_bk_l_05, bone2Name=dyn_skirt_bk_l_04, swing1LimitDegrees=6, swing2LimitDegrees=30, twistLimitDegrees=0}
 {bone1Name=dyn_skirt_bk_l_06, bone2Name=dyn_skirt_bk_l_05, swing1LimitDegrees=6, swing2LimitDegrees=30, twistLimitDegrees=0}
 {
  bone1Name=dyn_skirt_l_MSTR
  bone2Name=thigh_l
  swing1LimitDegrees=45
  swing1Motion=Free
  swing2LimitDegrees=45
  swing2Motion=Free
  twistLimitDegrees=45
  twistMotion=Free
 }
 {bone1Name=dyn_skirt_l_01, bone2Name=dyn_skirt_l_MSTR, swing1LimitDegrees=20, swing2LimitDegrees=20, twistLimitDegrees=0}
 {bone1Name=dyn_skirt_l_02, bone2Name=dyn_skirt_l_01, swing1LimitDegrees=20, swing2LimitDegrees=20, twistLimitDegrees=0}
 {bone1Name=dyn_skirt_l_03, bone2Name=dyn_skirt_l_02, swing1LimitDegrees=20, swing2LimitDegrees=20, twistLimitDegrees=0}
 {bone1Name=dyn_skirt_l_04, bone2Name=dyn_skirt_l_03, swing1LimitDegrees=20, swing2LimitDegrees=20, twistLimitDegrees=0}
 {bone1Name=dyn_skirt_l_05, bone2Name=dyn_skirt_l_04, swing1LimitDegrees=20, swing2LimitDegrees=20, twistLimitDegrees=0}
 {bone1Name=dyn_skirt_l_06, bone2Name=dyn_skirt_l_05, swing1LimitDegrees=20, swing2LimitDegrees=20, twistLimitDegrees=0}
 {
  bone1Name=dyn_skirt_front_r_MSTR
  bone2Name=thigh_r
  swing1LimitDegrees=45
  swing1Motion=Free
  swing2LimitDegrees=45
  swing2Motion=Free
  twistLimitDegrees=45
  twistMotion=Free
 }
 {bone1Name=dyn_skirt_front_r_01, bone2Name=dyn_skirt_front_r_MSTR, swing1LimitDegrees=20, swing2LimitDegrees=20, twistLimitDegrees=0}
 {bone1Name=dyn_skirt_front_r_02, bone2Name=dyn_skirt_front_r_01, swing1LimitDegrees=20, swing2LimitDegrees=20, twistLimitDegrees=0}
 {bone1Name=dyn_skirt_front_r_03, bone2Name=dyn_skirt_front_r_02, swing1LimitDegrees=20, swing2LimitDegrees=20, twistLimitDegrees=0}
 {bone1Name=dyn_skirt_front_r_04, bone2Name=dyn_skirt_front_r_03, swing1LimitDegrees=20, swing2LimitDegrees=20, twistLimitDegrees=0}
 {bone1Name=dyn_skirt_front_r_05, bone2Name=dyn_skirt_front_r_04, swing1LimitDegrees=20, swing2LimitDegrees=45, twistLimitDegrees=0}
 {bone1Name=dyn_skirt_front_r_06, bone2Name=dyn_skirt_front_r_05, swing1LimitDegrees=20, swing2LimitDegrees=45, twistLimitDegrees=0}
 {bone1Name=dyn_skirt_bk_r_01, bone2Name=dyn_skirt_bk_r_MSTR, swing1LimitDegrees=7, swing2LimitDegrees=20, twistLimitDegrees=0}
 {bone1Name=dyn_skirt_bk_r_02, bone2Name=dyn_skirt_bk_r_01, swing1LimitDegrees=6, swing2LimitDegrees=20, twistLimitDegrees=0}
 {bone1Name=dyn_skirt_bk_r_03, bone2Name=dyn_skirt_bk_r_02, swing1LimitDegrees=6, swing2LimitDegrees=20, twistLimitDegrees=0}
 {bone1Name=dyn_skirt_bk_r_04, bone2Name=dyn_skirt_bk_r_03, swing1LimitDegrees=6, swing2LimitDegrees=20, twistLimitDegrees=0}
 {bone1Name=dyn_skirt_bk_r_05, bone2Name=dyn_skirt_bk_r_04, swing1LimitDegrees=6, swing2LimitDegrees=30, twistLimitDegrees=0}
 {bone1Name=dyn_skirt_bk_r_06, bone2Name=dyn_skirt_bk_r_05, swing1LimitDegrees=6, swing2LimitDegrees=30, twistLimitDegrees=0}
 {
  bone1Name=dyn_skirt_r_MSTR
  bone2Name=thigh_r
  swing1LimitDegrees=45
  swing1Motion=Free
  swing2LimitDegrees=45
  swing2Motion=Free
  twistLimitDegrees=45
  twistMotion=Free
 }
 {bone1Name=dyn_skirt_r_01, bone2Name=dyn_skirt_r_MSTR, swing1LimitDegrees=20, swing2LimitDegrees=20, twistLimitDegrees=0}
 {bone1Name=dyn_skirt_r_02, bone2Name=dyn_skirt_r_01, swing1LimitDegrees=20, swing2LimitDegrees=20, twistLimitDegrees=0}
 {bone1Name=dyn_skirt_r_03, bone2Name=dyn_skirt_r_02, swing1LimitDegrees=20, swing2LimitDegrees=20, twistLimitDegrees=0}
 {bone1Name=dyn_skirt_r_04, bone2Name=dyn_skirt_r_03, swing1LimitDegrees=20, swing2LimitDegrees=20, twistLimitDegrees=0}
 {bone1Name=dyn_skirt_r_05, bone2Name=dyn_skirt_r_04, swing1LimitDegrees=20, swing2LimitDegrees=20, twistLimitDegrees=0}
 {bone1Name=dyn_skirt_r_06, bone2Name=dyn_skirt_r_05, swing1LimitDegrees=20, swing2LimitDegrees=20, twistLimitDegrees=0}
 {bone1Name=dyn_accessory_C_01, bone2Name=dyn_accessory_C_MSTR, swing1LimitDegrees=30, swing2LimitDegrees=45, twistLimitDegrees=0}
 {bone1Name=dyn_accessory_B_01, bone2Name=dyn_accessory_B_MSTR, swing1LimitDegrees=30, swing2LimitDegrees=45, twistLimitDegrees=0}
 {bone1Name=dyn_accessory_B_02, bone2Name=dyn_accessory_B_01, swing1LimitDegrees=20, swing2LimitDegrees=20, twistLimitDegrees=0}
 {bone1Name=dyn_accessory_A_01, bone2Name=dyn_accessory_A_MSTR, swing1LimitDegrees=15, swing2LimitDegrees=15, twistLimitDegrees=0}
 {bone1Name=dyn_accessory_A_02, bone2Name=dyn_accessory_A_01, swing1LimitDegrees=10, swing2LimitDegrees=10, twistLimitDegrees=0}
 {
  bone1Name=dyn_accessory_C_MSTR
  bone2Name=thigh_r
  swing1LimitDegrees=45
  swing1Motion=Free
  swing2LimitDegrees=45
  swing2Motion=Free
  twistLimitDegrees=45
  twistMotion=Free
 }
 {
  bone1Name=dyn_accessory_B_MSTR
  bone2Name=thigh_r
  swing1LimitDegrees=45
  swing1Motion=Free
  swing2LimitDegrees=45
  swing2Motion=Free
  twistLimitDegrees=45
  twistMotion=Free
 }
 {
  bone1Name=dyn_accessory_A_MSTR
  bone2Name=thigh_r
  swing1LimitDegrees=45
  swing1Motion=Free
  swing2LimitDegrees=45
  swing2Motion=Free
  twistLimitDegrees=45
  twistMotion=Free
 }
 {
  bone1Name=thigh_r
  bone2Name=pelvis
  swing1LimitDegrees=45
  swing1Motion=Free
  swing2LimitDegrees=45
  swing2Motion=Free
  twistLimitDegrees=45
  twistMotion=Free
 }
 {
  bone1Name=calf_r
  bone2Name=thigh_r
  swing1LimitDegrees=45
  swing1Motion=Free
  swing2LimitDegrees=45
  swing2Motion=Free
  twistLimitDegrees=45
  twistMotion=Free
 }
]
