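# Physics asset for the collar and necklace of the medium female player's Wayfare body, presumably for secondary motion on its dyn_ bones.
# Capsules cover five collar bones, two necklace bones, and the head, neck and upper spine.
# Each collar bone constrains directly to neck_01 with tight swing and zero twist limits.
# The necklace pendant bone has tight limits to its master, which joins spine_05 with free rotation, as do the head-to-spine links.
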PhysicsAsset /Game/Characters/Player/Female/Medium/Bodies/F_MED_Wayfare/Meshes/F_MED_Wayfare_Collar_Physics
bodies {
 dyn_collarA_l_01 [{shapeType=Capsule, length=6.1461, radius=2.2543, shapeName=dyn_collarA_l_01_capsule, center=<3.233 0 0>, rotation=<rot 90 0 0>}]
 dyn_collarA_r_01 [{shapeType=Capsule, length=6.1461, radius=2.2543, shapeName=dyn_collarA_r_01_capsule, center=<-3.233 0 0>, rotation=<rot 90 0 0>}]
 dyn_collarB_l_01 [{
   shapeType=Capsule
   length=8.7886
   radius=2.468
   shapeName=dyn_collarB_l_01_capsule
   center=<3.9769 0 -0.7211>
   rotation=<rot 82.361 -0.0002 22.6946>
  }]
 dyn_collarB_r_01 [{
   shapeType=Capsule
   length=8.7886
   radius=2.468
   shapeName=dyn_collarB_r_01_capsule
   center=<4.1432 0 0.6373>
   rotation=<rot 87.6504 179.9996 -159.398>
  }]
 dyn_collar_c_01 [{shapeType=Capsule, length=8.6783, radius=2.2543, shapeName=dyn_collar_c_01_capsule, center=<4.9833 0 -1.051>, rotation=<rot 85 0 0>}]
 dyn_necklace_MSTR [{shapeType=Capsule, length=0.505, radius=0.505, shapeName=dyn_necklace_MSTR_capsule, rotation=<rot 0 -90 0>}]
 dyn_necklace_c_01 [{shapeType=Capsule, length=2.1425, radius=4.8604, shapeName=dyn_necklace_c_01_capsule, center=<8.2072 -4 0>}]
 head [{shapeType=Capsule, length=9.1838, radius=4.8195, shapeName=head_capsule, center=<0.2838 -1.0032 0>, rotation=<rot 0 -96.251 90>}]
 neck_01 [{shapeType=Capsule, length=0.505, radius=0.505, shapeName=neck_01_capsule, rotation=<rot 0 -90 0>}]
 neck_02 [{shapeType=Capsule, length=0, radius=4.3794, shapeName=neck_02_capsule, rotation=<rot 0 -90 0>}]
 spine_05 [{shapeType=Capsule, length=11.5797, radius=11.0627, shapeName=spine_05_capsule}]
}
constraints [
 {
  bone1Name=head
  bone2Name=neck_02
  swing1LimitDegrees=45
  swing1Motion=Free
  swing2LimitDegrees=45
  swing2Motion=Free
  twistLimitDegrees=45
  twistMotion=Free
 }
 {bone1Name=dyn_collarA_l_01, bone2Name=neck_01, swing1LimitDegrees=20, swing2LimitDegrees=20, twistLimitDegrees=0}
 {bone1Name=dyn_collarB_l_01, bone2Name=neck_01, swing1LimitDegrees=5, swing2LimitDegrees=30, twistLimitDegrees=0}
 {bone1Name=dyn_collar_c_01, bone2Name=neck_01, swing1LimitDegrees=5, swing2LimitDegrees=30, twistLimitDegrees=0}
 {bone1Name=dyn_collarA_r_01, bone2Name=neck_01, swing1LimitDegrees=20, swing2LimitDegrees=20, twistLimitDegrees=0}
 {bone1Name=dyn_collarB_r_01, bone2Name=neck_01, swing1LimitDegrees=5, swing2LimitDegrees=30, twistLimitDegrees=0}
 {
  bone1Name=neck_02
  bone2Name=neck_01
  swing1LimitDegrees=45
  swing1Motion=Free
  swing2LimitDegrees=45
  swing2Motion=Free
  twistLimitDegrees=45
  twistMotion=Free
 }
 {bone1Name=dyn_necklace_c_01, bone2Name=dyn_necklace_MSTR, swing1LimitDegrees=20, swing2LimitDegrees=10, twistLimitDegrees=11}
 {
  bone1Name=neck_01
  bone2Name=spine_05
  swing1LimitDegrees=45
  swing1Motion=Free
  swing2LimitDegrees=45
  swing2Motion=Free
  twistLimitDegrees=45
  twistMotion=Free
 }
 {
  bone1Name=dyn_necklace_MSTR
  bone2Name=spine_05
  swing1LimitDegrees=45
  swing1Motion=Free
  swing2LimitDegrees=45
  swing2Motion=Free
  twistLimitDegrees=45
  twistMotion=Free
 }
]
