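# Post-process-style cosmetic AnimBP for the F_MED_Wayfare body, adding secondary motion to skirt, collar, cloak and accessories.
# It copies the attached parent's pose, applies corrective bone edits and Control Rig, then selects thigh-pinned sliding, crouch, front-end or standard RigidBody skirt setups.
# The cloak then simulates or locks to the spine depending on backbling, before accessory and collar fixes.
# The EventGraph derives these states and resets dynamics around skydiving.
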AnimBlueprint /BRCosmetics/Characters/Player/Female/Medium/Bodies/F_MED_Wayfare/Meshes/F_MED_Wayfare_AnimBP
target_skeleton=&/F_MED_Wayfare_Skeleton
graphs [
 AnimGraph {nodes [
   RigidBody!(ComponentPose, Alpha, OverrideWorldGravity:Vector=Get Gravity Override(), ExternalForce) -> (Pose) {
    overridePhysicsAsset=/Game/Characters/Player/Female/Medium/Bodies/F_MED_Wayfare/Meshes/F_MED_Wayfare_Collar_Physics
    simulationSpace=BaseBoneSpace
    alphaBoolBlend {blendOption=Linear}
    baseBoneRef {boneName=spine_05}
    componentAppliedLinearAccClamp=<6000 6000 6000>
    componentLinearAccScale=<2 1 0.3>
    componentLinearVelScale=<2 1 0.3>
    overrideWorldGravity {z=-980}
   }
   n1 = "Constraint - dyn_accessory_C_01 (PPP)"!(
    ComponentPose=Local To Component(LocalPose=Use cached pose 'PostSimCache'())
    Alpha=0.5
    ConstraintWeights_0=0.5
    ConstraintWeights_1:float=0.5
    ConstraintWeights_2:float=0
   ) -> (Pose) {
    alphaBoolBlend {blendOption=Linear}
    boneToModify {boneName=dyn_accessory_C_01}
    constraintSetup [{targetBone {boneName=dyn_skirt_front_r_02}}, {targetBone {boneName=dyn_skirt_r_02}}, {targetBone {boneName=thigh_r}}]
    constraintWeights [1,1,1]
   }
   n2 = "Constraint - dyn_accessory_A_02 (PPP)"!(ComponentPose=@n1, Alpha=0.5, ConstraintWeights_0=0.5, ConstraintWeights_1:float=0.5, ConstraintWeights_2:float=0) -> (Pose) {
    alphaBoolBlend {blendOption=Linear}
    boneToModify {boneName=dyn_accessory_A_02}
    constraintSetup [{targetBone {boneName=dyn_skirt_front_r_02}}, {targetBone {boneName=dyn_skirt_r_02}}, {targetBone {boneName=thigh_r}}]
    constraintWeights [1,1,1]
   }
   n3 = "Constraint - dyn_accessory_B_02 (PPP)"!(ComponentPose=@n2, Alpha=0.5, ConstraintWeights_0=0.5, ConstraintWeights_1:float=0.5, ConstraintWeights_2:float=0) -> (Pose) {
    alphaBoolBlend {blendOption=Linear}
    boneToModify {boneName=dyn_accessory_B_02}
    constraintSetup [{targetBone {boneName=dyn_skirt_front_r_02}}, {targetBone {boneName=dyn_skirt_r_02}}, {targetBone {boneName=thigh_r}}]
    constraintWeights [1,1,1]
   }
   n4 = Get bIsSkyDiving()
   n5 = "Transform (Modify) Bone - Bone: dyn_collar_c_01"!(ComponentPose=@n3, bAlphaBoolEnabled:bool=@n4, Translation="-1,0,0", Rotation="-18,-0.916,0", Scale="1,1,1") -> (Pose) {
    alphaInputType=Bool
    rotationMode=BMM_Additive
    rotationSpace=BCS_BoneSpace
    translationMode=BMM_Additive
    translationSpace=BCS_BoneSpace
    alphaBoolBlend {blendOption=Linear}
    boneToModify {boneName=dyn_collar_c_01}
    rotation {pitch=-18, yaw=-0.916}
    translation {x=-1}
   }
   n6 = "Transform (Modify) Bone - Bone: dyn_collarB_l_01"!(ComponentPose=@n5, bAlphaBoolEnabled:bool=@n4, Translation="0,0,0", Rotation="-18,-0.916,0", Scale="1,1,1") -> (Pose) {
    alphaInputType=Bool
    rotationMode=BMM_Additive
    rotationSpace=BCS_BoneSpace
    translationMode=BMM_Additive
    translationSpace=BCS_BoneSpace
    alphaBoolBlend {blendOption=Linear}
    boneToModify {boneName=dyn_collarB_l_01}
    rotation {pitch=-18, yaw=-0.916}
   }
   n7 = "Transform (Modify) Bone - Bone: dyn_collarB_r_01"!(ComponentPose=@n6, bAlphaBoolEnabled:bool=@n4, Translation="0,0,0", Rotation="18,-0.916,0", Scale="1,1,1") -> (Pose) {
    alphaInputType=Bool
    rotationMode=BMM_Additive
    rotationSpace=BCS_BoneSpace
    translationMode=BMM_Additive
    translationSpace=BCS_BoneSpace
    alphaBoolBlend {blendOption=Linear}
    boneToModify {boneName=dyn_collarB_r_01}
    rotation {pitch=18, yaw=-0.916}
   }
   n8 = "Transform (Modify) Bone - Bone: thigh_twist_01_r"!(ComponentPose=@n7, bAlphaBoolEnabled:bool=Get bIsInFrontEnd(), Translation="0,0,0", Rotation="0,0,30", Scale="1,1,1") -> (Pose) {
    alphaInputType=Bool
    rotationMode=BMM_Additive
    rotationSpace=BCS_BoneSpace
    translationMode=BMM_Additive
    translationSpace=BCS_BoneSpace
    alphaBoolBlend {blendOption=Linear}
    boneToModify {boneName=thigh_twist_01_r}
    rotation {roll=30}
   }
   Output Pose(Result=Component To Local(ComponentPose=@n8)) {layerGroup=None}
   n9 = Use cached pose 'CloakCache'()
   Save cached pose 'PostSimCache'(Pose=Component To Local(ComponentPose=Local To Component(LocalPose=@n9))) {cacheName=PostSimCache}
   n10 = "Constraint - dyn_collar_c_01 (P)"(ComponentPose=Local To Component(LocalPose=Copy Pose From Mesh() {bUseAttachedParent=true}), Alpha=0.5, ConstraintWeights_0=1) {
    alphaBoolBlend {blendOption=Linear}
    boneToModify {boneName=dyn_collar_c_01}
    constraintSetup [{targetBone {boneName=spine_05}}]
    constraintWeights [1]
   }
   n11 = "Constraint - dyn_collarA_l_01 (P)"(ComponentPose=@n10, Alpha=0.5, ConstraintWeights_0=1) {
    alphaBoolBlend {blendOption=Linear}
    boneToModify {boneName=dyn_collarA_l_01}
    constraintSetup [{targetBone {boneName=spine_05}}]
    constraintWeights [1]
   }
   n12 = "Constraint - dyn_collarB_l_01 (P)"(ComponentPose=@n11, Alpha=0.5, ConstraintWeights_0=1) {
    alphaBoolBlend {blendOption=Linear}
    boneToModify {boneName=dyn_collarB_l_01}
    constraintSetup [{targetBone {boneName=spine_05}}]
    constraintWeights [1]
   }
   n13 = "Constraint - dyn_collarA_r_01 (P)"(ComponentPose=@n12, Alpha=0.5, ConstraintWeights_0=1) {
    alphaBoolBlend {blendOption=Linear}
    boneToModify {boneName=dyn_collarA_r_01}
    constraintSetup [{targetBone {boneName=spine_05}}]
    constraintWeights [1]
   }
   n14 = "Constraint - dyn_collarB_r_01 (P)"(ComponentPose=@n13, Alpha=0.5, ConstraintWeights_0=1) {
    alphaBoolBlend {blendOption=Linear}
    boneToModify {boneName=dyn_collarB_r_01}
    constraintSetup [{targetBone {boneName=spine_05}}]
    constraintWeights [1]
   }
   n15 = Get head up() -> (head up:double)
   n16 = "Transform (Modify) Bone - Bone: dyn_collar_c_01"!(ComponentPose=@n14, Alpha:float=@n15, Translation="0,0,-10", Rotation="-30,0,0", Scale="1,1,1") -> (Pose) {
    rotationMode=BMM_Additive
    rotationSpace=BCS_BoneSpace
    alphaBoolBlend {blendOption=Linear}
    boneToModify {boneName=dyn_collar_c_01}
    rotation {pitch=-30}
    translation {z=-10}
   }
   n17 = "Transform (Modify) Bone - Bone: dyn_collarB_r_01"!(ComponentPose=@n16, Alpha:float=@n15, Translation="0,0,0", Rotation="30,0,0", Scale="1,1,1") -> (Pose) {
    rotationMode=BMM_Additive
    rotationSpace=BCS_BoneSpace
    alphaBoolBlend {blendOption=Linear}
    boneToModify {boneName=dyn_collarB_r_01}
    rotation {pitch=30}
   }
   n18 = "Constraint - dyn_skirt_bk_r_MSTR (PP)"!(ComponentPose=@n17, Alpha=1, ConstraintWeights_0=0.2, ConstraintWeights_1:float=0.8) -> (Pose) {
    alphaBoolBlend {blendOption=Linear}
    boneToModify {boneName=dyn_skirt_bk_r_MSTR}
    constraintSetup [{targetBone {boneName=thigh_r}}, {targetBone {boneName=pelvis}}]
    constraintWeights [1,1]
   }
   n19 = "Constraint - dyn_skirt_r_MSTR (PP)"!(ComponentPose=@n18, Alpha=1, ConstraintWeights_0=0.2, ConstraintWeights_1:float=0.8) -> (Pose) {
    alphaBoolBlend {blendOption=Linear}
    boneToModify {boneName=dyn_skirt_r_MSTR}
    constraintSetup [{targetBone {boneName=thigh_r}}, {targetBone {boneName=pelvis}}]
    constraintWeights [1,1]
   }
   n20 = "Constraint - dyn_skirt_bk_l_MSTR (PP)"!(ComponentPose=@n19, Alpha=1, ConstraintWeights_0=0.2, ConstraintWeights_1:float=0.8) -> (Pose) {
    alphaBoolBlend {blendOption=Linear}
    boneToModify {boneName=dyn_skirt_bk_l_MSTR}
    constraintSetup [{targetBone {boneName=thigh_l}}, {targetBone {boneName=pelvis}}]
    constraintWeights [1,1]
   }
   n21 = "Constraint - dyn_skirt_l_MSTR (PP)"!(ComponentPose=@n20, Alpha=1, ConstraintWeights_0=0.2, ConstraintWeights_1:float=0.8) -> (Pose) {
    alphaBoolBlend {blendOption=Linear}
    boneToModify {boneName=dyn_skirt_l_MSTR}
    constraintSetup [{targetBone {boneName=thigh_l}}, {targetBone {boneName=pelvis}}]
    constraintWeights [1,1]
   }
   n22 = Get SkirtAlpha() -> (SkirtAlpha:double)
   n23 = "Transform (Modify) Bone - Bone: dyn_skirt_front_l_MSTR"(ComponentPose=@n21, Alpha=@n22, Translation="0,0,0", Rotation="0,-30,0", Scale="1,1,1") {
    rotationMode=BMM_Additive
    rotationSpace=BCS_BoneSpace
    alphaBoolBlend {blendInTime=1, blendOption=Linear}
    boneToModify {boneName=dyn_skirt_front_l_MSTR}
    rotation {yaw=-30}
   }
   n24 = "Transform (Modify) Bone - Bone: dyn_skirt_bk_l_MSTR"(ComponentPose=@n23, Alpha=@n22, Translation="0,0,2", Rotation="10,0,0", Scale="1,1,1") {
    rotationMode=BMM_Additive
    rotationSpace=BCS_BoneSpace
    translationMode=BMM_Additive
    translationSpace=BCS_BoneSpace
    alphaBoolBlend {blendInTime=1, blendOption=Linear}
    boneToModify {boneName=dyn_skirt_bk_l_MSTR}
    rotation {pitch=10}
    translation {z=2}
   }
   n25 = "Transform (Modify) Bone - Bone: dyn_skirt_l_MSTR"(ComponentPose=@n24, Alpha=@n22, Translation="0,0,2", Rotation="10,0,0", Scale="1,1,1") {
    rotationMode=BMM_Additive
    rotationSpace=BCS_BoneSpace
    translationMode=BMM_Additive
    translationSpace=BCS_BoneSpace
    alphaBoolBlend {blendInTime=1, blendOption=Linear}
    boneToModify {boneName=dyn_skirt_l_MSTR}
    rotation {pitch=10}
    translation {z=2}
   }
   n26 = "Transform (Modify) Bone - Bone: dyn_skirt_front_r_MSTR"(ComponentPose=@n25, Alpha=@n22, Translation="0,0,0", Rotation="0,-30,0", Scale="1,1,1") {
    rotationMode=BMM_Additive
    rotationSpace=BCS_BoneSpace
    alphaBoolBlend {blendInTime=1, blendOption=Linear}
    boneToModify {boneName=dyn_skirt_front_r_MSTR}
    rotation {yaw=-30}
   }
   n27 = "Transform (Modify) Bone - Bone: dyn_skirt_bk_r_MSTR"(ComponentPose=@n26, Alpha=@n22, Translation="0,0,2", Rotation="10,0,0", Scale="1,1,1") {
    rotationMode=BMM_Additive
    rotationSpace=BCS_BoneSpace
    translationMode=BMM_Additive
    translationSpace=BCS_BoneSpace
    alphaBoolBlend {blendInTime=1, blendOption=Linear}
    boneToModify {boneName=dyn_skirt_bk_r_MSTR}
    rotation {pitch=10}
    translation {z=2}
   }
   n28 = "Transform (Modify) Bone - Bone: dyn_skirt_r_MSTR"(ComponentPose=@n27, Alpha=@n22, Translation="0,0,2", Rotation="10,-20,0", Scale="1,1,1") {
    rotationMode=BMM_Additive
    rotationSpace=BCS_BoneSpace
    translationMode=BMM_Additive
    translationSpace=BCS_BoneSpace
    alphaBoolBlend {blendInTime=1, blendOption=Linear}
    boneToModify {boneName=dyn_skirt_r_MSTR}
    rotation {pitch=10, yaw=-20}
    translation {z=2}
   }
   n29 = Anim Dynamics - Bone: dyn_necklace_c_01(ComponentPose=@n28, Alpha=0, ExternalForce="0,0,0") {
    angularDampingOverride=1
    angularSpringConstant=20000
    bAngularSpring=true
    bOverrideAngularDamping=true
    bUseGravityOverride=true
    gravityScale=0.35
    simulationSpace=BoneRelative
    alphaBoolBlend {blendOption=Linear}
    boundBone {boneName=dyn_necklace_c_01}
    componentAppliedLinearAccClamp=<6000 6000 6000>
    componentLinearAccScale=<3 3 3>
    componentLinearVelScale=<3 3 3>
    gravityOverride {z=-980}
    physicsBodyDefinitions [{
      boundBone {boneName=dyn_necklace_c_01}
      boxExtents=<10 10 10>
      constraintSetup {angularLimitsMax=<0 0 30>, angularLimitsMin=<0 0 -30>, angularTarget=<0 0 0>}
      localJointOffset=<0 0 0>
     }]
    relativeSpaceBone {boneName=neck_01}
   }
   n30 = Control Rig(Source=Component To Local(ComponentPose=@n29)) {
    alphaBoolBlend {blendOption=Linear}
    controlRigAssetReference {blueprintRigClass=&/F_MED_WayfareRig_Master_ControlRig_new1.F_MED_WayfareRig_Master_ControlRig_new1_C}
   }
   Save cached pose 'PreSim Cache'(Pose=@n30) {cacheName=PreSim Cache}
   n31 = Get cloak Acc Scale Vals() -> (cloak Acc Scale Vals:Vector)
   n32 = RigidBody!(
    ComponentPose=Local To Component(LocalPose=Use cached pose 'defaultPose'())
    bAlphaBoolEnabled:bool
    OverrideWorldGravity=Get Gravity Override()
    ExternalForce
    ComponentLinearAccScale:Vector=@n31
    ComponentLinearVelScale:Vector=@n31
   ) -> (Pose) {
    alphaInputType=Bool
    bOverrideWorldGravity=true
    overridePhysicsAsset=&/F_MED_Wayfare_Cloak_Physics
    simulationSpace=BaseBoneSpace
    alphaBoolBlend {blendOption=Linear}
    baseBoneRef {boneName=spine_04}
    componentAppliedLinearAccClamp=<6000 6000 6000>
    componentLinearAccScale=<0.05 0.5 0.5>
    componentLinearVelScale=<0.05 0.5 0.5>
    overrideWorldGravity {z=-980}
    simSpaceSettings {worldAlpha=0.15}
   }
   n33 = "Constraint - dyn_cloak_01 (P)"(ComponentPose=Local To Component(LocalPose=Use cached pose 'defaultPose'()), Alpha=1, ConstraintWeights_0=1) {
    alphaBoolBlend {blendOption=Linear}
    boneToModify {boneName=dyn_cloak_01}
    constraintSetup [{targetBone {boneName=spine_05}}]
    constraintWeights [1]
   }
   n34 = "Constraint - dyn_cloak_02 (P)"(ComponentPose=@n33, Alpha=1, ConstraintWeights_0=1) {
    alphaBoolBlend {blendOption=Linear}
    boneToModify {boneName=dyn_cloak_02}
    constraintSetup [{targetBone {boneName=spine_03}}]
    constraintWeights [1]
   }
   n35 = "Constraint - dyn_cloak_03 (P)"(ComponentPose=@n34, Alpha=1, ConstraintWeights_0=1) {
    alphaBoolBlend {blendOption=Linear}
    boneToModify {boneName=dyn_cloak_03}
    constraintSetup [{targetBone {boneName=spine_02}}]
    constraintWeights [1]
   }
   Save cached pose 'CloakCache'(Pose=Blend Poses by bool(
     BlendPose_0=Component To Local(ComponentPose=@n32)
     BlendPose_1=Component To Local(ComponentPose=@n35)
     bActiveValue=Get is Not Wearing Backbling() -> (is Not Wearing Backbling:bool)
    ) {blendType=Linear}) {cacheName=CloakCache}
   n36 = RigidBody!(
    ComponentPose=Local To Component(LocalPose=Use cached pose 'PreSim Cache'())
    Alpha:float
    OverrideWorldGravity=Get Gravity Override()
    ExternalForce
   ) -> (Pose) {
    bOverrideWorldGravity=true
    overridePhysicsAsset=&/F_MED_Wayfare_Physics
    simulationSpace=BaseBoneSpace
    alphaBoolBlend {blendOption=Linear}
    baseBoneRef {boneName=root}
    componentAppliedLinearAccClamp=<6000 6000 6000>
    componentLinearAccScale=<0.05 0.05 0.15>
    componentLinearVelScale=<0.05 0.05 0.15>
    overrideWorldGravity {z=-980}
   }
   n37 = "Constraint - dyn_skirt_front_r_03 (P)"(ComponentPose=@n36, Alpha=1, ConstraintWeights_0=1) {
    alphaBoolBlend {blendOption=Linear}
    alphaScaleBiasClamp {bInterpResult=true, interpSpeedDecreasing=5, interpSpeedIncreasing=7}
    boneToModify {boneName=dyn_skirt_front_r_03}
    constraintSetup [{targetBone {boneName=thigh_r}}]
    constraintWeights [1]
   }
   n38 = "Constraint - dyn_skirt_front_r_03 (P)"(ComponentPose=@n37, Alpha=1, ConstraintWeights_0=1) {
    alphaBoolBlend {blendOption=Linear}
    alphaScaleBiasClamp {bClampResult=true, bInterpResult=true, clampMax=0.6, interpSpeedDecreasing=5, interpSpeedIncreasing=7}
    boneToModify {boneName=dyn_skirt_front_r_03}
    constraintSetup [{targetBone {boneName=thigh_r}}]
    constraintWeights [1]
   }
   n39 = "Constraint - dyn_skirt_front_l_03 (P)"(ComponentPose=@n38, Alpha=1, ConstraintWeights_0=1) {
    alphaBoolBlend {blendOption=Linear}
    alphaScaleBiasClamp {bClampResult=true, bInterpResult=true, clampMax=0.6, interpSpeedDecreasing=5, interpSpeedIncreasing=7}
    boneToModify {boneName=dyn_skirt_front_l_03}
    constraintSetup [{targetBone {boneName=thigh_l}}]
    constraintWeights [1]
   }
   n40 = Look At - Bone: dyn_skirt_l_MSTR(ComponentPose=@n39, Alpha=1) {
    alphaBoolBlend {blendOption=Linear}
    alphaScaleBiasClamp {bClampResult=true, bInterpResult=true, clampMax=0.7, interpSpeedIncreasing=15}
    boneToModify {boneName=dyn_skirt_l_MSTR}
    lookAtLocation=<-72.014 -6.7988 -31.755>
    lookAtTarget {boneReference {boneName=thigh_l}}
    lookAt_Axis {axis=<1 0>}
   }
   n41 = Look At - Bone: dyn_skirt_r_MSTR(ComponentPose=@n40, Alpha=1) {
    alphaBoolBlend {blendOption=Linear}
    alphaScaleBiasClamp {bClampResult=true, bInterpResult=true, clampMax=0.7, interpSpeedIncreasing=15}
    boneToModify {boneName=dyn_skirt_r_MSTR}
    lookAtLocation=<-33.0988 -2.6089 -6.9493>
    lookAtTarget {boneReference {boneName=thigh_r}}
    lookAt_Axis {axis=<1 0>}
   }
   n42 = RigidBody(ComponentPose=Local To Component(LocalPose=Use cached pose 'PreSim Cache'()), OverrideWorldGravity=Get Gravity Override()) {
    bOverrideWorldGravity=true
    overridePhysicsAsset=&/F_MED_Wayfare_FRONTEND_Physics
    simulationSpace=BaseBoneSpace
    alphaBoolBlend {blendOption=Linear}
    baseBoneRef {boneName=root}
    componentAppliedLinearAccClamp=<6000 6000 6000>
    componentLinearAccScale=<0.05 0.05 0.15>
    componentLinearVelScale=<0.05 0.05 0.15>
    overrideWorldGravity {z=-980}
   }
   n43 = "Constraint - dyn_skirt_front_r_03 (P)"(ComponentPose=@n42, Alpha=1, ConstraintWeights_0=1) {
    alphaBoolBlend {blendOption=Linear}
    alphaScaleBiasClamp {bClampResult=true, bInterpResult=true, clampMax=0.5, interpSpeedDecreasing=5, interpSpeedIncreasing=7}
    boneToModify {boneName=dyn_skirt_front_r_03}
    constraintSetup [{targetBone {boneName=thigh_r}}]
    constraintWeights [1]
   }
   n44 = Look At - Bone: dyn_skirt_r_MSTR(ComponentPose=@n43, Alpha=1) {
    alphaBoolBlend {blendOption=Linear}
    alphaScaleBiasClamp {bClampResult=true, bInterpResult=true, clampMax=0.7, interpSpeedIncreasing=15}
    boneToModify {boneName=dyn_skirt_r_MSTR}
    lookAtLocation=<-33.1028 -2.5939 1.9479>
    lookAtTarget {boneReference {boneName=thigh_r}}
    lookAt_Axis {axis=<1 0>}
   }
   n45 = "Transform (Modify) Bone - Bone: dyn_skirt_front_l_MSTR"!(ComponentPose=@n44, Alpha=0, Translation="0,0,2", Rotation="15,0,0") -> (Pose) {
    rotationMode=BMM_Replace
    rotationSpace=BCS_BoneSpace
    translationMode=BMM_Additive
    translationSpace=BCS_BoneSpace
    alphaBoolBlend {blendOption=Linear}
    alphaScaleBiasClamp {bClampResult=true, bInterpResult=true, clampMax=0.8, interpSpeedIncreasing=15}
    boneToModify {boneName=dyn_skirt_front_l_MSTR}
    rotation {pitch=15}
    translation {z=2}
   }
   n46 = "Constraint - dyn_skirt_front_l_03 (P)"(ComponentPose=@n45, Alpha=1, ConstraintWeights_0=1) {
    alphaBoolBlend {blendOption=Linear}
    alphaScaleBiasClamp {bClampResult=true, bInterpResult=true, clampMax=0.8, interpSpeedDecreasing=5, interpSpeedIncreasing=7}
    boneToModify {boneName=dyn_skirt_front_l_03}
    constraintSetup [{targetBone {boneName=thigh_l}}]
    constraintWeights [1]
   }
   n47 = Look At - Bone: dyn_skirt_l_MSTR(ComponentPose=@n46, Alpha=1) {
    alphaBoolBlend {blendOption=Linear}
    alphaScaleBiasClamp {bClampResult=true, bInterpResult=true, clampMax=0.7, interpSpeedIncreasing=15}
    boneToModify {boneName=dyn_skirt_l_MSTR}
    lookAtLocation=<-72.0191 -6.757 -21.6939>
    lookAtTarget {boneReference {boneName=thigh_l}}
    lookAt_Axis {axis=<1 0>}
   }
   n48 = "Constraint - dyn_skirt_bk_l_04 (P)"(ComponentPose=@n47, Alpha=1, ConstraintWeights_0=1) {
    alphaBoolBlend {blendOption=Linear}
    alphaScaleBiasClamp {bClampResult=true, bInterpResult=true, clampMax=0.2, interpSpeedDecreasing=5, interpSpeedIncreasing=7}
    boneToModify {boneName=dyn_skirt_bk_l_04}
    constraintSetup [{targetBone {boneName=calf_l}}]
    constraintWeights [1]
   }
   n49 = "Constraint - dyn_skirt_bk_r_04 (P)"(ComponentPose=@n48, Alpha=1, ConstraintWeights_0=1) {
    alphaBoolBlend {blendOption=Linear}
    alphaScaleBiasClamp {bClampResult=true, bInterpResult=true, clampMax=0.2, interpSpeedDecreasing=5, interpSpeedIncreasing=7}
    boneToModify {boneName=dyn_skirt_bk_r_04}
    constraintSetup [{targetBone {boneName=calf_r}}]
    constraintWeights [1]
   }
   n50 = Blend Poses by bool(
    BlendPose_0=Component To Local(ComponentPose=@n49)
    BlendPose_1=Component To Local(ComponentPose=@n41)
    BlendTime_0=0.2
    BlendTime_1=1
    bActiveValue=Get bIsInFrontEnd()
   ) {blendType=Linear}
   n51 = "Transform (Modify) Bone - Bone: dyn_skirt_front_l_01"(
    ComponentPose=Local To Component(LocalPose=Use cached pose 'PreSim Cache'())
    Alpha=1
    Translation="0,3,1"
    Rotation="18,-40,0"
    Scale="1,1,1"
   ) {
    rotationMode=BMM_Additive
    rotationSpace=BCS_BoneSpace
    translationMode=BMM_Additive
    translationSpace=BCS_BoneSpace
    alphaBoolBlend {blendOption=Linear}
    boneToModify {boneName=dyn_skirt_front_l_01}
    rotation {pitch=18, yaw=-40}
    translation {y=3, z=1}
   }
   n52 = "Transform (Modify) Bone - Bone: dyn_skirt_front_l_02"(ComponentPose=@n51, Alpha=1, Translation="0,0,0", Rotation="-40,-20,0", Scale="1,1,1") {
    rotationMode=BMM_Additive
    rotationSpace=BCS_BoneSpace
    translationMode=BMM_Additive
    translationSpace=BCS_BoneSpace
    alphaBoolBlend {blendOption=Linear}
    boneToModify {boneName=dyn_skirt_front_l_02}
    rotation {pitch=-40, yaw=-20}
   }
   n53 = "Transform (Modify) Bone - Bone: dyn_skirt_front_l_03"(ComponentPose=@n52, Alpha=1, Translation="0,0,0", Rotation="0,-15,0", Scale="1,1,1") {
    rotationMode=BMM_Additive
    rotationSpace=BCS_BoneSpace
    translationMode=BMM_Additive
    translationSpace=BCS_BoneSpace
    alphaBoolBlend {blendOption=Linear}
    boneToModify {boneName=dyn_skirt_front_l_03}
    rotation {yaw=-15}
   }
   n54 = "Transform (Modify) Bone - Bone: dyn_skirt_l_01"(ComponentPose=@n53, Alpha=1, Translation="-2,2,3", Rotation="20,-60,0", Scale="1,1,1") {
    rotationMode=BMM_Additive
    rotationSpace=BCS_BoneSpace
    translationMode=BMM_Additive
    translationSpace=BCS_BoneSpace
    alphaBoolBlend {blendOption=Linear}
    boneToModify {boneName=dyn_skirt_l_01}
    rotation {pitch=20, yaw=-60}
    translation=<-2 2 3>
   }
   n55 = "Transform (Modify) Bone - Bone: dyn_skirt_l_02"(ComponentPose=@n54, Alpha=1, Translation="0,0,0", Rotation="0,-20,0", Scale="1,1,1") {
    rotationMode=BMM_Additive
    rotationSpace=BCS_BoneSpace
    translationMode=BMM_Additive
    translationSpace=BCS_BoneSpace
    alphaBoolBlend {blendOption=Linear}
    boneToModify {boneName=dyn_skirt_l_02}
    rotation {yaw=-20}
   }
   n56 = "Transform (Modify) Bone - Bone: dyn_skirt_l_03"(ComponentPose=@n55, Alpha=1, Translation="0,0,0", Rotation="0,-10,0", Scale="1,1,1") {
    rotationMode=BMM_Additive
    rotationSpace=BCS_BoneSpace
    translationMode=BMM_Additive
    translationSpace=BCS_BoneSpace
    alphaBoolBlend {blendOption=Linear}
    boneToModify {boneName=dyn_skirt_l_03}
    rotation {yaw=-10}
   }
   n57 = "Transform (Modify) Bone - Bone: dyn_skirt_bk_l_01"(ComponentPose=@n56, Alpha=1, Translation="0,0,0", Rotation="45,20,0", Scale="1,1,1") {
    rotationMode=BMM_Additive
    rotationSpace=BCS_BoneSpace
    translationMode=BMM_Additive
    translationSpace=BCS_BoneSpace
    alphaBoolBlend {blendOption=Linear}
    boneToModify {boneName=dyn_skirt_bk_l_01}
    rotation {pitch=45, yaw=20}
   }
   n58 = "Transform (Modify) Bone - Bone: dyn_skirt_bk_r_01"(ComponentPose=@n57, Alpha=1, Translation="0,0,0", Rotation="45,40,0", Scale="1,1,1") {
    rotationMode=BMM_Additive
    rotationSpace=BCS_BoneSpace
    translationMode=BMM_Additive
    translationSpace=BCS_BoneSpace
    alphaBoolBlend {blendOption=Linear}
    boneToModify {boneName=dyn_skirt_bk_r_01}
    rotation {pitch=45, yaw=40}
   }
   n59 = "Transform (Modify) Bone - Bone: dyn_skirt_l_03"(ComponentPose=@n58, Alpha=0, Translation="0,0,0", Rotation="0,-10,0", Scale="1,1,1") {
    rotationMode=BMM_Additive
    rotationSpace=BCS_BoneSpace
    translationMode=BMM_Additive
    translationSpace=BCS_BoneSpace
    alphaBoolBlend {blendOption=Linear}
    boneToModify {boneName=dyn_skirt_l_03}
    rotation {yaw=-10}
   }
   n60 = RigidBody(ComponentPose=@n59, OverrideWorldGravity=Get Gravity Override()) {
    bOverrideWorldGravity=true
    overridePhysicsAsset=&/F_MED_Wayfare_CrouchingPhysics
    simulationSpace=BaseBoneSpace
    alphaBoolBlend {blendOption=Linear}
    baseBoneRef {boneName=root}
    componentAppliedLinearAccClamp=<6000 6000 6000>
    componentLinearAccScale=<0.05 0.05 0.05>
    componentLinearVelScale=<0.05 0.05 0.05>
    overrideWorldGravity {z=-980}
   }
   n61 = Component To Local(ComponentPose=@n60)
   n62 = Get Is Crouching() -> (Is Crouching:bool)
   n63 = "Constraint - dyn_skirt_front_l_01 (P)"(ComponentPose=Local To Component(LocalPose=Use cached pose 'PreSim Cache'()), Alpha=1, ConstraintWeights_0=1) {
    alphaBoolBlend {blendOption=Linear}
    boneToModify {boneName=dyn_skirt_front_l_01}
    constraintSetup [{targetBone {boneName=thigh_l}}]
    constraintWeights [1]
   }
   n64 = "Constraint - dyn_skirt_front_l_03 (P)"(ComponentPose=@n63, Alpha=1, ConstraintWeights_0=1) {
    alphaBoolBlend {blendOption=Linear}
    boneToModify {boneName=dyn_skirt_front_l_03}
    constraintSetup [{targetBone {boneName=thigh_l}}]
    constraintWeights [1]
   }
   n65 = "Constraint - dyn_skirt_front_l_04 (P)"(ComponentPose=@n64, Alpha=1, ConstraintWeights_0=1) {
    alphaBoolBlend {blendOption=Linear}
    boneToModify {boneName=dyn_skirt_front_l_04}
    constraintSetup [{targetBone {boneName=calf_l}}]
    constraintWeights [1]
   }
   n66 = "Constraint - dyn_skirt_l_01 (P)"(ComponentPose=@n65, Alpha=1, ConstraintWeights_0=1) {
    alphaBoolBlend {blendOption=Linear}
    boneToModify {boneName=dyn_skirt_l_01}
    constraintSetup [{targetBone {boneName=thigh_l}}]
    constraintWeights [1]
   }
   n67 = "Constraint - dyn_skirt_l_03 (P)"(ComponentPose=@n66, Alpha=1, ConstraintWeights_0=1) {
    alphaBoolBlend {blendOption=Linear}
    boneToModify {boneName=dyn_skirt_l_03}
    constraintSetup [{targetBone {boneName=thigh_l}}]
    constraintWeights [1]
   }
   n68 = "Constraint - dyn_skirt_l_04 (P)"(ComponentPose=@n67, Alpha=1, ConstraintWeights_0=1) {
    alphaBoolBlend {blendOption=Linear}
    boneToModify {boneName=dyn_skirt_l_04}
    constraintSetup [{targetBone {boneName=calf_l}}]
    constraintWeights [1]
   }
   n69 = "Constraint - dyn_skirt_bk_l_01 (P)"(ComponentPose=@n68, Alpha=1, ConstraintWeights_0=1) {
    alphaBoolBlend {blendOption=Linear}
    boneToModify {boneName=dyn_skirt_bk_l_01}
    constraintSetup [{targetBone {boneName=thigh_l}}]
    constraintWeights [1]
   }
   n70 = "Constraint - dyn_skirt_bk_l_03 (P)"(ComponentPose=@n69, Alpha=1, ConstraintWeights_0=1) {
    alphaBoolBlend {blendOption=Linear}
    boneToModify {boneName=dyn_skirt_bk_l_03}
    constraintSetup [{targetBone {boneName=thigh_l}}]
    constraintWeights [1]
   }
   n71 = "Constraint - dyn_skirt_bk_l_04 (P)"(ComponentPose=@n70, Alpha=1, ConstraintWeights_0=1) {
    alphaBoolBlend {blendOption=Linear}
    boneToModify {boneName=dyn_skirt_bk_l_04}
    constraintSetup [{targetBone {boneName=calf_l}}]
    constraintWeights [1]
   }
   n72 = "Constraint - dyn_skirt_front_r_01 (P)"(ComponentPose=@n71, Alpha=1, ConstraintWeights_0=1) {
    alphaBoolBlend {blendOption=Linear}
    boneToModify {boneName=dyn_skirt_front_r_01}
    constraintSetup [{targetBone {boneName=thigh_r}}]
    constraintWeights [1]
   }
   n73 = "Constraint - dyn_skirt_front_r_03 (P)"(ComponentPose=@n72, Alpha=1, ConstraintWeights_0=1) {
    alphaBoolBlend {blendOption=Linear}
    boneToModify {boneName=dyn_skirt_front_r_03}
    constraintSetup [{targetBone {boneName=thigh_r}}]
    constraintWeights [1]
   }
   n74 = "Constraint - dyn_skirt_front_r_04 (P)"(ComponentPose=@n73, Alpha=1, ConstraintWeights_0=1) {
    alphaBoolBlend {blendOption=Linear}
    boneToModify {boneName=dyn_skirt_front_r_04}
    constraintSetup [{targetBone {boneName=calf_r}}]
    constraintWeights [1]
   }
   n75 = "Constraint - dyn_skirt_r_01 (P)"(ComponentPose=@n74, Alpha=1, ConstraintWeights_0=1) {
    alphaBoolBlend {blendOption=Linear}
    boneToModify {boneName=dyn_skirt_r_01}
    constraintSetup [{targetBone {boneName=thigh_r}}]
    constraintWeights [1]
   }
   n76 = "Constraint - dyn_skirt_r_03 (P)"(ComponentPose=@n75, Alpha=1, ConstraintWeights_0=1) {
    alphaBoolBlend {blendOption=Linear}
    boneToModify {boneName=dyn_skirt_r_03}
    constraintSetup [{targetBone {boneName=thigh_r}}]
    constraintWeights [1]
   }
   n77 = "Constraint - dyn_skirt_r_04 (P)"(ComponentPose=@n76, Alpha=1, ConstraintWeights_0=1) {
    alphaBoolBlend {blendOption=Linear}
    boneToModify {boneName=dyn_skirt_r_04}
    constraintSetup [{targetBone {boneName=calf_r}}]
    constraintWeights [1]
   }
   n78 = "Constraint - dyn_skirt_bk_l_01 (P)"(ComponentPose=@n77, Alpha=1, ConstraintWeights_0=1) {
    alphaBoolBlend {blendOption=Linear}
    boneToModify {boneName=dyn_skirt_bk_l_01}
    constraintSetup [{targetBone {boneName=thigh_r}}]
    constraintWeights [1]
   }
   n79 = "Constraint - dyn_skirt_bk_l_03 (P)"(ComponentPose=@n78, Alpha=1, ConstraintWeights_0=1) {
    alphaBoolBlend {blendOption=Linear}
    boneToModify {boneName=dyn_skirt_bk_l_03}
    constraintSetup [{targetBone {boneName=thigh_r}}]
    constraintWeights [1]
   }
   n80 = "Constraint - dyn_skirt_bk_l_04 (P)"(ComponentPose=@n79, Alpha=1, ConstraintWeights_0=1) {
    alphaBoolBlend {blendOption=Linear}
    boneToModify {boneName=dyn_skirt_bk_l_04}
    constraintSetup [{targetBone {boneName=calf_r}}]
    constraintWeights [1]
   }
   Save cached pose 'defaultPose'(Pose=Blend Poses by bool(
     BlendPose_0=Component To Local(ComponentPose=@n80)
     BlendPose_1=Blend Poses by bool(BlendPose_0=@n61, BlendPose_1=@n50, BlendTime_0=0.2, BlendTime_1=1, bActiveValue=@n62) {blendType=Linear}
     BlendTime_0=0.2
     BlendTime_1=1
     bActiveValue=Get bIsSliding()
    ) {blendType=Linear}) {cacheName=defaultPose}
   Get isNotCrouching() -> (isNotCrouching:bool)
  ]}
 EventGraph {nodes [
   n1 = Event OnEndSkydiving()
   ResetDynamics(execute=[@n1.then, @n2.then, @n3.then], InTeleportType=TeleportPhysics)
   > ResetDynamics(InTeleportType=ResetPhysics)
   n2 = Event OnBeginSkydiving()
   Event OnEndCrouching()
   > Set endCrouching(endCrouching:bool=true) -> (Output_Get:bool)
   > n3 = Set crouch alpha(crouch alpha:double=1) -> (Output_Get:double)
   Event OnBeginCrouching()
   > Set endCrouching(endCrouching=false)
   > Set crouch alpha(crouch alpha=1)
   Event BlueprintInitializeAnimation()
   > Set Skel_Mesh(Skel_Mesh=GetOwnerSkeletalMeshComponent())
   n4 = Event BlueprintUpdateAnimation()
   > Set Delta Time X(Delta Time X:double=@n4.DeltaTimeX) -> (Output_Get:double)
   > n5 = Sequence!(execute) -> (then_0, then_1, then_2:Exec, then_3:Exec, then_4:Exec, then_5:Exec, then_6:Exec, then_7:Exec, then_8:Exec, then_9:Exec)
   > n6 = Gravity Override State Function DEPRECATED(
    CharacterPartAnimInstance=Self-Reference()
    GameTimeInSeconds=Get Delta Time X() -> (Delta Time X:double)
    GravityOverrideParams=Get Gravity Override Params()
   )
   > Set Gravity Override(Gravity Override=@n6)
   n7 = CheckForBackbling(self=Self-Reference())
   n8 = Branch(execute=@n5.then_1, Condition=OR Boolean!(A=@n7, B=@n7.bIsBackblingACape, C:bool=@n7.bIsBackblingLong) -> (ReturnValue))
   > Set is Not Wearing Backbling(is Not Wearing Backbling:bool=false) -> (Output_Get:bool)
   > n9 = Branch(Condition=Get SpecialBackpackEquipped() -> (SpecialBackpackEquipped:bool))
   > Set is Not Wearing Backbling(is Not Wearing Backbling=true)
   Set is Not Wearing Backbling(execute=@n9.else, is Not Wearing Backbling=false)
   Set is Not Wearing Backbling(execute=@n8.else, is Not Wearing Backbling=true)
   n10 = Branch(execute=@n5.then_2, Condition=Get Is Crouching())
   n11 = BreakTransform(InTransform=GetDeltaTransformFromRefPose(self=GetOwnerSkeletalMeshComponent(), BoneName=head, BaseName=spine_05))
   > Set head up(head up:double=MapRangeClamped(Value=BreakRotator(InRot=@n11.Rotation).Yaw, InRangeA=-55, InRangeB=-75)) -> (Output_Get:double)
   Set head up(execute=@n10.else, head up=0)
   n12 = Get bIsCrouching()
   Set Is Crouching(execute=@n5.then_3, Is Crouching:bool=@n12) -> (Output_Get:bool)
   > n13 = Branch(Condition=@n12)
   > Set isNotCrouching(isNotCrouching:bool=false) -> (Output_Get:bool)
   Set isNotCrouching(execute=@n13.else, isNotCrouching=true)
   > n14 = Branch(Condition=Get bIsInFrontEnd())
   > Set isNotCrouching(isNotCrouching=false)
   Set isNotCrouching(execute=@n14.else, isNotCrouching=true)
   Set head rot Back(
    execute=@n5.then_4
    head rot Back:double=GetAlphaFromReferenceTransformDelta(SkeletalMeshComponent=Get Skel_Mesh(), BoneName=head, BaseName=neck_02, InputRange=(X=-45,Y=-30), OutputRange=(X=1,Y=0), TwistAxis="1,0,0")
   ) -> (Output_Get:double)
   n15 = Cast To FortPlayerAnimInstance(execute=@n5.then_5, Object=GetAnimInstance(self=GetOwnerSkeletalMeshComponent()))
   > Set Alt_weapon(Alt_weapon:bool=false) -> (Output_Get:bool)
   > Set isPistol(isPistol=false)
   > n16 = Switch on EFortWeaponCoreAnimation!(execute, Selection:EFortWeaponCoreAnimation Enum=Get WeaponCoreAnim(self=@n15)) -> (
    Melee:Exec
    Pistol:Exec
    Shotgun:Exec
    PaperBlueprint:Exec
    Rifle:Exec
    MeleeOneHand:Exec
    MachinePistol:Exec
    RocketLauncher:Exec
    GrenadeLauncher:Exec
    GoingCommando:Exec
    AssaultRifle:Exec
    TacticalShotgun:Exec
    SniperRifle:Exec
    TrapPlacement:Exec
    ShoulderLauncher:Exec
    AbilityDecoTool:Exec
    Crossbow:Exec
    C4:Exec
    RemoteControl:Exec
    DualWield:Exec
    AR_BullPup:Exec
    AR_ForwardGrip:Exec
    MedPackPaddles:Exec
    SMG_P90:Exec
    AR_DrumGun:Exec
    Consumable_Small:Exec
    Consumable_Large:Exec
    Balloon:Exec
    MountedTurret:Exec
    CreativeTool:Exec
    ExplosiveBow:Exec
    AshtonIndigo:Exec
    AshtonChicago:Exec
    MeleeDualWield:Exec
    Unarmed:Exec
   )
   n17 = Branch(execute=[@n16.Melee, @n16.MeleeOneHand])
   > Set isMelee(isMelee:bool=true) -> (Output_Get:bool)
   > n18 = Branch(Condition=Get bIsSprinting())
   > Set isSprintingMelee(isSprintingMelee:bool=true) -> (Output_Get:bool)
   Set isSprintingMelee(execute=@n18.else, isSprintingMelee=false)
   Set isMelee(execute=@n17.else, isMelee=false)
   n19 = Branch(execute=@n16.Pistol)
   > Set Alt_weapon(Alt_weapon=true)
   > Set isPistol(isPistol=true)
   Set isPistol(execute=@n19.else, isPistol=false)
   n20 = Branch(execute=@n5.then_6, Condition=Get bIsSkyDiving())
   > Set cloak Acc Scale Vals(cloak Acc Scale Vals:Vector="0.05,0.01,0.01") -> (Output_Get:Vector)
   Set cloak Acc Scale Vals(execute=@n20.else, cloak Acc Scale Vals="0.05,0.5,0.5")
   n21 = GetPartSkeletalMeshComponent()
   Set isWearingCustomBackpack(
    execute=@n5.then_7
    isWearingCustomBackpack:bool="Equal (Object)"(
     A=FindChildCommponentFromAttachSocket!(
      self:Custom Character Part Anim Instance Object Reference
      Parent:Skeletal Mesh Component Object Reference=@n21
      bIncludeAllDescendants:bool=false
      AttachSocket:Name=cape
     ) -> (ReturnValue:Skeletal Mesh Component Object Reference)
     B=/BRCosmetics/Athena/Items/Cosmetics/Backpacks/BID_A_061_WayfareFemale
    )
   ) -> (Output_Get:bool)
   n22 = Branch(
    execute=@n5.then_8
    Condition=BackpackItemDefinitionHasAnyMetaTag!(Tags:Array of Gameplay Tag Structures=Make Array("[0]"=(TagName="Cosmetics.MetaTags.WayfareFemale")), bRequiresEquipped:bool=false) -> (ReturnValue:bool)
   )
   > Set SpecialBackpackEquipped(SpecialBackpackEquipped:bool=true) -> (Output_Get:bool)
   Set SpecialBackpackEquipped(execute=@n22.else, SpecialBackpackEquipped=false)
   n23 = NOT Boolean(A=Get bIsCrouching())
   n24 = NOT Boolean(A=Get bIsSkyDiving())
   Set SkirtAlpha(
    execute=@n5.then_9
    SkirtAlpha:double=float * float(A="To Float (Boolean)"(InBool=@n23), B="To Float (Boolean)"(InBool=@n24)) {is=title K2Node_PromotableOperator}
   ) -> (Output_Get:double)
  ]}
]
variables [
 Alt_weapon {type=boolean}
 cloak Acc Scale Vals {type=Vector}
 crouch alpha {type=number}
 Delta Time X {type=number}
 endCrouching {type=boolean}
 Gravity Override {type=Vector}
 Gravity Override Params {type=GravityOverrideParamsStruct}
 head rot Back {type=number}
 head up {type=number}
 Is Crouching {type=boolean}
 is Not Wearing Backbling {type=boolean}
 Is Wearing Backbling {type=boolean}
 isMelee {type=boolean}
 isNotCrouching {type=boolean}
 isPistol {type=boolean}
 isSprintingMelee {type=boolean}
 isWearingCustomBackpack {type=boolean}
 look Up ADS {type=number}
 LookDown {type=number}
 LookUp {type=number}
 LookUpRifle {type=number}
 Skel_Mesh {type=SkeletalMeshComponent}
 SkirtAlpha {type=number}
 SpecialBackpackEquipped {type=boolean}
]
