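# Physics asset for the female medium Wayfare body, named as the collider set for its cape, presumably to keep cape simulation off the character.
# Capsules cover neck, spine, pelvis, arms, hands, thighs and calves, plus selected front, side and back dynamic skirt bones.
# Constraints chain bodies along the skeleton, hanging front and side skirt bodies from thighs and back ones from pelvis, with free rotation at some spine and shoulder joints.
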PhysicsAsset /BRCosmetics/Characters/Player/Female/Medium/Bodies/F_MED_Wayfare/Meshes/F_MED_Wayfare_CapeColliders
bodies {
 calf_l [{shapeType=Capsule, length=28.2698, radius=12.8372, center=<-27.0722 3.086 -1.082>, rotation=<rot 0 81.1707 -89.8557>}]
 calf_r [{shapeType=Capsule, length=28.2698, radius=12.8372, center=<27.0722 -3.086 1.082>, rotation=<rot 0 81.1707 -89.8557>}]
 clavicle_l [{shapeType=Capsule, length=15.7761, radius=7.6833, shapeName=clavicle_l_capsule, rotation=<rot -85.1535 179.9992 -179.9992>}]
 clavicle_r [{shapeType=Capsule, length=19.0197, radius=9.5098}]
 dyn_skirt_bk_l_03 [{shapeType=Capsule, length=23.2639, radius=9.4512, center=<5.6945 0 0>, rotation=<rot 90 0 0>}]
 dyn_skirt_bk_l_05 [{shapeType=Capsule, length=23.2639, radius=9.4512, center=<5.6945 0 0>, rotation=<rot 90 0 0>}]
 dyn_skirt_bk_r_03 [{shapeType=Capsule, length=23.2639, radius=9.4512, center=<-6.847 0 0>, rotation=<rot 90 0 0>}]
 dyn_skirt_bk_r_05 [{shapeType=Capsule, length=23.2639, radius=9.4512, center=<-6.847 0 0>, rotation=<rot 90 0 0>}]
 dyn_skirt_front_l_04 [{shapeType=Capsule, length=23.2639, radius=9.4512, center=<5.6945 0 0>, rotation=<rot 90 0 0>}]
 dyn_skirt_front_r_04 [{shapeType=Capsule, length=23.2639, radius=9.4512, center=<-6.847 0 0>, rotation=<rot 90 0 0>}]
 dyn_skirt_l_03 [{shapeType=Capsule, length=23.2639, radius=9.4512, center=<5.6945 0 0>, rotation=<rot 90 0 0>}]
 dyn_skirt_l_05 [{shapeType=Capsule, length=23.2639, radius=9.4512, center=<5.6945 0 0>, rotation=<rot 90 0 0>}]
 dyn_skirt_r_03 [{shapeType=Capsule, length=23.2639, radius=9.4512, center=<-6.847 0 0>, rotation=<rot 90 0 0>}]
 dyn_skirt_r_05 [{shapeType=Capsule, length=23.2639, radius=9.4512, center=<-6.847 0 0>, rotation=<rot 90 0 0>}]
 hand_l [{shapeType=Capsule, length=8.5769, radius=6.6522, center=<7.8355 -2.5748 0.1112>, rotation=<rot 0 86.8749 -62.3709>}]
 hand_r [{shapeType=Capsule, length=8.6112, radius=6.1901, center=<-8.1729 2.1632 -0.4284>, rotation=<rot 0 81.0488 -75.6535>}]
 lowerarm_l [{shapeType=Capsule, length=15.4371, radius=9.9416, center=<13.3788 0 0>, rotation=<rot 90 0 0>}]
 lowerarm_r [{shapeType=Capsule, length=13.703, radius=7.2047, center=<-11.4492 -0.5459 -1.4574>, rotation=<rot 0 -86.5513 -79.432>}]
 neck_01 [{shapeType=Capsule, length=10.3058, radius=8.4822, center=<6.06 1.0987 0.0636>, rotation=<rot 0 9.2781 -88.0687>}]
 pelvis [{shapeType=Capsule, length=6.261, radius=13.5408}]
 spine_01 [{shapeType=Capsule, length=2.71, radius=12.2408, shapeName=spine_01_capsule}]
 spine_02 [{shapeType=Capsule, length=0.6588, radius=12.2408, shapeName=spine_02_capsule}]
 spine_03 [{shapeType=Capsule, length=3.4074, radius=12.2408, shapeName=spine_03_capsule}]
 spine_04 [{shapeType=Capsule, length=4.9488, radius=12.2408}]
 spine_05 [{shapeType=Capsule, length=8.6861, radius=12.2408}]
 thigh_l [{shapeType=Capsule, length=33.8759, radius=12.8553, center=<-28.0162 0.8501 1.9493>, rotation=<rot 0.0006 -89.5569 -95.6962>}]
 thigh_r [{shapeType=Capsule, length=33.8759, radius=12.8553, center=<28.0162 -0.8501 -1.9493>, rotation=<rot 0.0006 -89.5569 84.3038>}]
 upperarm_l [{shapeType=Capsule, length=15.5011, radius=8.2593, center=<13.1958 0 0>, rotation=<rot 90 0 0>}]
 upperarm_r [{shapeType=Capsule, length=8.7783, radius=5.6222, shapeName=upperarm_r_capsule}]
}
constraints [
 {bone1Name=spine_04, bone2Name=spine_03, swing1LimitDegrees=45, swing2LimitDegrees=45, twistLimitDegrees=45}
 {bone1Name=spine_05, bone2Name=spine_04, swing1LimitDegrees=45, swing2LimitDegrees=45, twistLimitDegrees=45}
 {bone1Name=neck_01, bone2Name=spine_05, swing1LimitDegrees=45, swing2LimitDegrees=45, twistLimitDegrees=45}
 {bone1Name=upperarm_l, bone2Name=clavicle_l, swing1LimitDegrees=45, swing2LimitDegrees=45, twistLimitDegrees=45}
 {bone1Name=lowerarm_l, bone2Name=upperarm_l, swing1LimitDegrees=45, swing2LimitDegrees=45, twistLimitDegrees=45}
 {bone1Name=hand_l, bone2Name=lowerarm_l, swing1LimitDegrees=45, swing2LimitDegrees=45, twistLimitDegrees=45}
 {bone1Name=clavicle_r, bone2Name=spine_05, swing1LimitDegrees=45, swing2LimitDegrees=45, twistLimitDegrees=45}
 {bone1Name=lowerarm_r, bone2Name=upperarm_r, swing1LimitDegrees=45, swing2LimitDegrees=45, twistLimitDegrees=45}
 {bone1Name=hand_r, bone2Name=lowerarm_r, swing1LimitDegrees=45, swing2LimitDegrees=45, twistLimitDegrees=45}
 {bone1Name=thigh_l, bone2Name=pelvis, swing1LimitDegrees=45, swing2LimitDegrees=45, twistLimitDegrees=45}
 {bone1Name=calf_l, bone2Name=thigh_l, swing1LimitDegrees=45, swing2LimitDegrees=45, twistLimitDegrees=45}
 {bone1Name=dyn_skirt_front_l_04, bone2Name=thigh_l, swing1LimitDegrees=45, swing2LimitDegrees=45, twistLimitDegrees=45}
 {bone1Name=dyn_skirt_l_03, bone2Name=thigh_l, swing1LimitDegrees=45, swing2LimitDegrees=45, twistLimitDegrees=45}
 {bone1Name=dyn_skirt_l_05, bone2Name=dyn_skirt_l_03, swing1LimitDegrees=45, swing2LimitDegrees=45, twistLimitDegrees=45}
 {bone1Name=calf_r, bone2Name=thigh_r, swing1LimitDegrees=45, swing2LimitDegrees=45, twistLimitDegrees=45}
 {bone1Name=dyn_skirt_front_r_04, bone2Name=thigh_r, swing1LimitDegrees=45, swing2LimitDegrees=45, twistLimitDegrees=45}
 {bone1Name=dyn_skirt_r_03, bone2Name=thigh_r, swing1LimitDegrees=45, swing2LimitDegrees=45, twistLimitDegrees=45}
 {bone1Name=dyn_skirt_r_05, bone2Name=dyn_skirt_r_03, swing1LimitDegrees=45, swing2LimitDegrees=45, twistLimitDegrees=45}
 {bone1Name=dyn_skirt_bk_r_03, bone2Name=pelvis, swing1LimitDegrees=45, swing2LimitDegrees=45, twistLimitDegrees=45}
 {bone1Name=dyn_skirt_bk_r_05, bone2Name=dyn_skirt_bk_r_03, swing1LimitDegrees=45, swing2LimitDegrees=45, twistLimitDegrees=45}
 {bone1Name=dyn_skirt_bk_l_03, bone2Name=pelvis, swing1LimitDegrees=45, swing2LimitDegrees=45, twistLimitDegrees=45}
 {bone1Name=dyn_skirt_bk_l_05, bone2Name=dyn_skirt_bk_l_03, swing1LimitDegrees=45, swing2LimitDegrees=45, twistLimitDegrees=45}
 {
  bone1Name=spine_03
  bone2Name=spine_02
  swing1LimitDegrees=45
  swing1Motion=Free
  swing2LimitDegrees=45
  swing2Motion=Free
  twistLimitDegrees=45
  twistMotion=Free
 }
 {
  bone1Name=spine_01
  bone2Name=pelvis
  swing1LimitDegrees=45
  swing1Motion=Free
  swing2LimitDegrees=45
  swing2Motion=Free
  twistLimitDegrees=45
  twistMotion=Free
 }
 {
  bone1Name=spine_02
  bone2Name=spine_01
  swing1LimitDegrees=45
  swing1Motion=Free
  swing2LimitDegrees=45
  swing2Motion=Free
  twistLimitDegrees=45
  twistMotion=Free
 }
 {
  bone1Name=clavicle_l
  bone2Name=spine_05
  swing1LimitDegrees=45
  swing1Motion=Free
  swing2LimitDegrees=45
  swing2Motion=Free
  twistLimitDegrees=45
  twistMotion=Free
 }
 {
  bone1Name=upperarm_r
  bone2Name=clavicle_r
  swing1LimitDegrees=45
  swing1Motion=Free
  swing2LimitDegrees=45
  swing2Motion=Free
  twistLimitDegrees=45
  twistMotion=Free
 }
 {bone1Name=thigh_r, bone2Name=pelvis, swing1LimitDegrees=45, swing2LimitDegrees=45, twistLimitDegrees=45}
]
